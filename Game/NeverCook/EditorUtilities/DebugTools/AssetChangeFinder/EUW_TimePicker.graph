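# Editor utility date/time picker widget from the AssetChangeFinder debug tools, with an optional label and NOW button.
# PreConstruct applies StartingDate silently; NOW applies current UTC time.
# SetDateTime fills the day, month, year, hour and minute spin boxes, updates a preview text and optionally broadcasts onTimeChanged.
# Spin box commits broadcast, then call SetDateTime with getDateTime's rebuilt value, so the event fires twice per edit.
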EditorUtilityWidgetBlueprint /Game/NeverCook/EditorUtilities/DebugTools/AssetChangeFinder/EUW_TimePicker
graphs [
 EventGraph {nodes [
   "On Clicked (btnNow)"() {is=title K2Node_ComponentBoundEvent}
   > SetDateTime!(execute, self, InDateTime:Date Time Structure=UTC Now!() -> (ReturnValue:Date Time Structure), AllowEvent:bool=true) -> (then)
   "On Value Committed (spinDay)"!() -> (OutputDelegate, then, InValue:float, CommitMethod:ETextCommit Enum=Default) {is=title K2Node_ComponentBoundEvent}
   > UpdateTimeInternal()
   "On Value Committed (spinHour)"!() -> (OutputDelegate, then, InValue:float, CommitMethod:ETextCommit Enum=Default) {is=title K2Node_ComponentBoundEvent}
   > UpdateTimeInternal()
   "On Value Committed (spinMinute)"!() -> (OutputDelegate, then, InValue:float, CommitMethod:ETextCommit Enum=Default) {is=title K2Node_ComponentBoundEvent}
   > UpdateTimeInternal()
   "On Value Committed (SpinMonth)"!() -> (OutputDelegate, then, InValue:float, CommitMethod:ETextCommit Enum=Default) {is=title K2Node_ComponentBoundEvent}
   > UpdateTimeInternal()
   "On Value Committed (spinYear)"!() -> (OutputDelegate, then, InValue:float, CommitMethod:ETextCommit Enum=Default) {is=title K2Node_ComponentBoundEvent}
   > UpdateTimeInternal()
   Event PreConstruct()
   > SetDateTime(InDateTime=Get StartingDate() -> (StartingDate:Date Time Structure), AllowEvent=false)
   n1 = Get Label() -> (Label:String)
   n2 = Get txtLabel() -> (txtLabel:Common Text Object Reference)
   > "SetText (Text)"(self=@n2, InText="To Text (String)"(InString=@n1))
   > SetVisibility(self=@n2, InVisibility=Select(Option 0=HitTestInvisible, Option 1=Collapsed, Index=IsEmpty(InString=@n1)))
   > SetVisibility(
    self=Get btnNow() -> (btnNow:Editor Utility Button Object Reference)
    InVisibility=Select(Option 0=Hidden, Option 1=Visible, Index=Get ShowNowButton() -> (ShowNowButton:bool))
   )
  ]}
 SetDateTime {nodes [
   n1 = SetDateTime!() -> (then, InDateTime:Date Time Structure, AllowEvent:bool=true) {is=title K2Node_FunctionEntry}
   n2 = BreakDateTime!(InDateTime:Date Time Structure=@n1) -> (Year:int=0, Month:int=0, Day:int=0, Hour:int=0, Minute:int=0, Second:int=0, Millisecond:int=0)
   > Set Value!(
    execute
    Value:float="To Float (Integer)"(InInt=@n2)
    self:Spin Box Object Reference=Get spinYear() -> (spinYear:Editor Utility Spin Box Object Reference)
   ) -> (then, Output_Get:float)
   > Set Value(Value="To Float (Integer)"(InInt=@n2.Month), self=Get SpinMonth() -> (SpinMonth:Editor Utility Spin Box Object Reference))
   > Set Value(Value="To Float (Integer)"(InInt=@n2.Day), self=Get spinDay() -> (spinDay:Editor Utility Spin Box Object Reference))
   > Set Value(Value="To Float (Integer)"(InInt=@n2.Hour), self=Get spinHour() -> (spinHour:Editor Utility Spin Box Object Reference))
   > Set Value(Value="To Float (Integer)"(InInt=@n2.Minute), self=Get spinMinute() -> (spinMinute:Editor Utility Spin Box Object Reference))
   > SetDateTimeValue!(
    execute
    self:Common Date Time Text Block Object Reference=Get CommonDateTimeTextBlock_0() -> (CommonDateTimeTextBlock_0:Common Date Time Text Block Object Reference)
    InDateTime:Date Time Structure=Get InDateTime() -> (InDateTime:Date Time Structure)
    bShowAsCountdown:bool=false
    InRefreshDelay:float=0
   ) -> (then)
   > Branch(Condition=Get AllowEvent() -> (AllowEvent:bool))
   > Call On Time Changed!(execute:Exec, self:EUW Time Picker Object Reference) -> (then:Exec) {is=title K2Node_CallDelegate}
  ]}
 UpdateTimeInternal {nodes [
   UpdateTimeInternal() {is=title K2Node_FunctionEntry}
   > Call On Time Changed()
   > SetDateTime(InDateTime=getDateTime!() -> (DateTime:Date Time Structure), AllowEvent=true) {is=title K2Node_CallFunction}
  ]}
 getDateTime {nodes [
   getDateTime() {is=title K2Node_FunctionEntry}
   n1 = Get Value!(self:Spin Box Object Reference=Get spinYear()) -> (Value:float)
   n2 = Get Value(self=Get SpinMonth())
   n3 = Get Value(self=Get spinDay())
   n4 = Get Value(self=Get spinHour())
   n5 = Get Value(self=Get spinMinute())
   > Return Node!(
    execute
    DateTime:Date Time Structure=MakeDateTime!(
     Year:int=Truncate(A=@n1)
     Month:int=Truncate(A=@n2)
     Day:int=Truncate(A=@n3)
     Hour:int=Truncate(A=@n4)
     Minute:int=Truncate(A=@n5)
     Second:int=0
     Millisecond:int=0
    ) -> (ReturnValue:Date Time Structure)
   )
   SetDateTimeValue(InDateTime=Get InDateTime(), bShowAsCountdown=false, InRefreshDelay=0)
  ]}
 onTimeChanged {nodes [onTimeChanged() {is=title K2Node_FunctionEntry}]}
]
variables [
 Label {type=string, default=Dude}
 onTimeChanged {default=()}
 ShowNowButton {type=boolean}
 StartingDate {type=DateTime, default=0001.01.01-00.00.00}
]
widgets {
 info {rootWidgetClass=&Overlay, widgetCount=20}
 widgets [
  Overlay_0 {widgetClassPath=&Overlay, widget=Overlay_0}
  Border_0 {
   widgetClassPath=&Border
   parent=Overlay_0
   widget=Border_0
   properties {
    background {
     bIsDynamicallyLoaded=false
     drawAs=Border
     imageType=FullColor
     mirroring=NoMirror
     resourceName=None
     resourceObject=null
     tiling=NoTile
     imageSize=<64 64>
     margin=<ltrb 0.4 0.4 0.4 0.4>
     outlineSettings {
      bUseBrushTransparency=false
      roundingType=HalfHeightRadius
      width=0
      color {colorUseRule=UseColor_Specified, specifiedColor=<rgba 0 0 0 0>}
      cornerRadii=<0 0 0 0>
     }
     tintColor {colorUseRule=UseColor_Specified, specifiedColor=<rgba 1 1 1 1>}
     uVRegion {bIsValid=false, max=<0 0>, min=<0 0>}
    }
    brushColor=<rgba 0 0 0 1>
    padding=<ltrb 0 0 0 0>
   }
   slot {class=OverlaySlot}
  }
  StackBox_0 {
   widgetClassPath=&StackBox
   parent=Border_0
   widget=StackBox_0
   properties {orientation=Orient_Vertical}
   slot {class=BorderSlot, properties {padding=<ltrb 0 0 0 0>}}
  }
  Border_1 {
   widgetClassPath=&Border
   parent=StackBox_0
   widget=Border_1
   properties {brushColor=<rgba 0 0 0 0.5>, padding=<ltrb 8 2 4 2>}
   slot {class=StackBoxSlot}
  }
  HorizontalBox_0 {
   widgetClassPath=&HorizontalBox
   parent=Border_1
   widget=HorizontalBox_0
   slot {class=BorderSlot, properties {padding=<ltrb 8 2 4 2>}}
  }
  txtLabel {
   widgetClassPath=&CommonTextBlock
   bIsVariable=true
   parent=HorizontalBox_0
   widget=txtLabel
   properties {
    font {
     bForceMonospaced=false
     bMaterialIsStencil=false
     fontMaterial=null
     fontObject=&NotoSans
     letterSpacing=0
     monospacedWidth=1
     size=9
     skewAmount=0
     typefaceFontName=NotoSans-Regular
     outlineSettings {
      bApplyOutlineToDropShadows=false
      bMiteredCorners=false
      bSeparateFillAlpha=false
      outlineMaterial=null
      outlineSize=0
      outlineColor=<rgba 0 0 0 1>
     }
    }
    shadowOffset=<0 0>
   }
   slot {class=HorizontalBoxSlot, properties {verticalAlignment=VAlign_Center}}
  }
  btnNow {
   widgetClassPath=&EditorUtilityButton
   parent=HorizontalBox_0
   widget=btnNow
   properties {
    toolTipText=Set to current Date/Time
    widgetStyle {
     clickedSlateSound {resourceObject=null}
     disabled {
      bIsDynamicallyLoaded=false
      drawAs=RoundedBox
      imageType=NoImage
      mirroring=NoMirror
      resourceName=None
      resourceObject=null
      tiling=NoTile
      imageSize=<0 0>
      margin=<ltrb 0 0 0 0>
      outlineSettings {
       bUseBrushTransparency=false
       roundingType=FixedRadius
       width=1
       color {colorUseRule=UseColor_ColorTable, specifiedColor=<rgba 0 0 0 0>}
       cornerRadii=<4 4 4 4>
      }
      tintColor {colorUseRule=UseColor_Specified, specifiedColor=<rgba 0.0395 0.0395 0.0395 1>}
      uVRegion {bIsValid=false, max=<0 0>, min=<0 0>}
     }
     disabledForeground {colorUseRule=UseColor_ColorTable, specifiedColor=<rgba 0 0 0 0>}
     hovered {
      bIsDynamicallyLoaded=false
      drawAs=RoundedBox
      imageType=NoImage
      mirroring=NoMirror
      resourceName=None
      resourceObject=null
      tiling=NoTile
      imageSize=<32 32>
      margin=<ltrb 0 0 0 0>
      outlineSettings {
       bUseBrushTransparency=false
       roundingType=FixedRadius
       width=1
       color {colorUseRule=UseColor_ColorTable, specifiedColor=<rgba 0 0 0 0>}
       cornerRadii=<4 4 4 4>
      }
      tintColor {colorUseRule=UseColor_Specified, specifiedColor=<rgba 0.0953 0.0953 0.0953 1>}
      uVRegion {bIsValid=false, max=<0 0>, min=<0 0>}
     }
     hoveredForeground {colorUseRule=UseColor_ColorTable, specifiedColor=<rgba 0 0 0 0>}
     hoveredSlateSound {resourceObject=null}
     normal {
      bIsDynamicallyLoaded=false
      drawAs=RoundedBox
      imageType=NoImage
      mirroring=NoMirror
      resourceName=None
      resourceObject=null
      tiling=NoTile
      imageSize=<32 32>
      margin=<ltrb 0 0 0 0>
      outlineSettings {
       bUseBrushTransparency=false
       roundingType=FixedRadius
       width=1
       color {colorUseRule=UseColor_ColorTable, specifiedColor=<rgba 0 0 0 0>}
       cornerRadii=<4 4 4 4>
      }
      tintColor {colorUseRule=UseColor_Specified, specifiedColor=<rgba 0.0395 0.0395 0.0395 1>}
      uVRegion {bIsValid=false, max=<0 0>, min=<0 0>}
     }
     normalForeground {colorUseRule=UseColor_ColorTable, specifiedColor=<rgba 0 0 0 0>}
     normalPadding=<ltrb 5 2 5 2>
     pressed {
      bIsDynamicallyLoaded=false
      drawAs=RoundedBox
      imageType=NoImage
      mirroring=NoMirror
      resourceName=None
      resourceObject=null
      tiling=NoTile
      imageSize=<32 32>
      margin=<ltrb 0 0 0 0>
      outlineSettings {
       bUseBrushTransparency=false
       roundingType=FixedRadius
       width=1
       color {colorUseRule=UseColor_ColorTable, specifiedColor=<rgba 0 0 0 0>}
       cornerRadii=<4 4 4 4>
      }
      tintColor {colorUseRule=UseColor_Specified, specifiedColor=<rgba 0.0284 0.0284 0.0284 1>}
      uVRegion {bIsValid=false, max=<0 0>, min=<0 0>}
     }
     pressedForeground {colorUseRule=UseColor_ColorTable, specifiedColor=<rgba 0 0 0 0>}
     pressedPadding=<ltrb 5 2 5 2>
     pressedSlateSound {resourceObject=null}
    }
   }
   slot {class=HorizontalBoxSlot, properties {horizontalAlignment=HAlign_Right, size {sizeRule=Fill, value=1}}}
  }
  txtLabel_1 {
   widgetClassPath=&CommonTextBlock
   bIsVariable=true
   parent=btnNow
   widget=txtLabel_1
   properties {
    text=NOW
    font {
     bForceMonospaced=false
     bMaterialIsStencil=false
     fontMaterial=null
     fontObject=&NotoSans
     letterSpacing=0
     monospacedWidth=1
     size=9
     skewAmount=0
     typefaceFontName=NotoSans-Regular
     outlineSettings {
      bApplyOutlineToDropShadows=false
      bMiteredCorners=false
      bSeparateFillAlpha=false
      outlineMaterial=null
      outlineSize=0
      outlineColor=<rgba 0 0 0 1>
     }
    }
    shadowOffset=<0 0>
   }
   slot {class=ButtonSlot, properties {horizontalAlignment=HAlign_Fill, padding=<ltrb 0 0 0 0>}}
  }
  VerticalBox_0 {
   widgetClassPath=&VerticalBox
   bIsVariable=true
   parent=StackBox_0
   widget=VerticalBox_0
   slot {class=StackBoxSlot, properties {padding=<ltrb 5 5 5 5>}}
  }
  HorizontalBox_1 {widgetClassPath=&HorizontalBox, parent=VerticalBox_0, widget=HorizontalBox_1, slot {class=VerticalBoxSlot}}
  spinDay {
   widgetClassPath=/Script/Blutility.EditorUtilitySpinBox
   bIsVariable=true
   parent=HorizontalBox_1
   widget=spinDay
   properties {
    bAlwaysUsesDeltaSnap=true
    bOverride_MaxSliderValue=true
    bOverride_MaxValue=true
    bOverride_MinSliderValue=true
    bOverride_MinValue=true
    delta=1
    justification=Center
    maxFractionalDigits=0
    maxSliderValue=31
    maxValue=31
    minDesiredWidth=30
    minFractionalDigits=0
    minSliderValue=1
    minValue=1
    toolTipText=Day
    value=31
    foregroundColor {colorUseRule=UseColor_Specified, specifiedColor=<rgba 1 1 1 1>}
   }
   slot {class=HorizontalBoxSlot}
  }
  txtLabel_4 {
   widgetClassPath=&CommonTextBlock
   bIsVariable=true
   parent=HorizontalBox_1
   widget=txtLabel_4
   properties {
    text=/
    font {
     bForceMonospaced=false
     bMaterialIsStencil=false
     fontMaterial=null
     fontObject=&NotoSans
     letterSpacing=0
     monospacedWidth=1
     size=13.5
     skewAmount=0
     typefaceFontName=NotoSans-Regular
     outlineSettings {
      bApplyOutlineToDropShadows=false
      bMiteredCorners=false
      bSeparateFillAlpha=false
      outlineMaterial=null
      outlineSize=0
      outlineColor=<rgba 0 0 0 1>
     }
    }
    shadowOffset=<0 0>
   }
   slot {class=HorizontalBoxSlot, properties {verticalAlignment=VAlign_Center}}
  }
  SpinMonth {
   widgetClassPath=&EditorUtilitySpinBox
   bIsVariable=true
   parent=HorizontalBox_1
   widget=SpinMonth
   properties {
    bAlwaysUsesDeltaSnap=true
    bOverride_MaxSliderValue=true
    bOverride_MaxValue=true
    bOverride_MinSliderValue=true
    bOverride_MinValue=true
    delta=1
    justification=Center
    maxFractionalDigits=0
    maxSliderValue=12
    maxValue=12
    minDesiredWidth=30
    minFractionalDigits=0
    minSliderValue=1
    minValue=1
    toolTipText=Month
    value=12
    foregroundColor {colorUseRule=UseColor_Specified, specifiedColor=<rgba 1 1 1 1>}
   }
   slot {class=HorizontalBoxSlot}
  }
  txtLabel_3 {
   widgetClassPath=&CommonTextBlock
   bIsVariable=true
   parent=HorizontalBox_1
   widget=txtLabel_3
   properties {
    text=/
    font {
     bForceMonospaced=false
     bMaterialIsStencil=false
     fontMaterial=null
     fontObject=&NotoSans
     letterSpacing=0
     monospacedWidth=1
     size=13.5
     skewAmount=0
     typefaceFontName=NotoSans-Regular
     outlineSettings {
      bApplyOutlineToDropShadows=false
      bMiteredCorners=false
      bSeparateFillAlpha=false
      outlineMaterial=null
      outlineSize=0
      outlineColor=<rgba 0 0 0 1>
     }
    }
    shadowOffset=<0 0>
   }
   slot {class=HorizontalBoxSlot, properties {verticalAlignment=VAlign_Center}}
  }
  spinYear {
   widgetClassPath=&EditorUtilitySpinBox
   bIsVariable=true
   parent=HorizontalBox_1
   widget=spinYear
   properties {
    bAlwaysUsesDeltaSnap=true
    bOverride_MaxSliderValue=true
    bOverride_MaxValue=true
    bOverride_MinSliderValue=true
    bOverride_MinValue=true
    delta=1
    justification=Center
    maxFractionalDigits=0
    maxSliderValue=2030
    maxValue=2050
    minDesiredWidth=50
    minFractionalDigits=0
    minSliderValue=2018
    minValue=2000
    toolTipText=Year
    value=2015
    foregroundColor {colorUseRule=UseColor_Specified, specifiedColor=<rgba 1 1 1 1>}
   }
   slot {class=HorizontalBoxSlot}
  }
  HorizontalBox {widgetClassPath=&HorizontalBox, parent=VerticalBox_0, widget=HorizontalBox, slot {class=VerticalBoxSlot}}
  spinHour {
   widgetClassPath=&EditorUtilitySpinBox
   bIsVariable=true
   parent=HorizontalBox
   widget=spinHour
   properties {
    bAlwaysUsesDeltaSnap=true
    bOverride_MaxSliderValue=true
    bOverride_MaxValue=true
    bOverride_MinSliderValue=true
    bOverride_MinValue=true
    delta=1
    justification=Center
    maxFractionalDigits=0
    maxSliderValue=24
    maxValue=24
    minDesiredWidth=30
    minFractionalDigits=0
    minSliderValue=1
    minValue=1
    toolTipText="Hour (1-24)"
    value=12
    foregroundColor {colorUseRule=UseColor_Specified, specifiedColor=<rgba 1 1 1 1>}
   }
   slot {class=HorizontalBoxSlot}
  }
  txtLabel_2 {
   widgetClassPath=&CommonTextBlock
   bIsVariable=true
   parent=HorizontalBox
   widget=txtLabel_2
   properties {
    text=:
    font {
     bForceMonospaced=false
     bMaterialIsStencil=false
     fontMaterial=null
     fontObject=&NotoSans
     letterSpacing=0
     monospacedWidth=1
     size=13.5
     skewAmount=0
     typefaceFontName=NotoSans-Regular
     outlineSettings {
      bApplyOutlineToDropShadows=false
      bMiteredCorners=false
      bSeparateFillAlpha=false
      outlineMaterial=null
      outlineSize=0
      outlineColor=<rgba 0 0 0 1>
     }
    }
    shadowOffset=<0 0>
   }
   slot {class=HorizontalBoxSlot, properties {verticalAlignment=VAlign_Center}}
  }
  spinMinute {
   widgetClassPath=&EditorUtilitySpinBox
   bIsVariable=true
   parent=HorizontalBox
   widget=spinMinute
   properties {
    bAlwaysUsesDeltaSnap=true
    bOverride_MaxSliderValue=true
    bOverride_MaxValue=true
    bOverride_MinSliderValue=true
    bOverride_MinValue=true
    delta=1
    justification=Center
    maxFractionalDigits=0
    maxSliderValue=59
    maxValue=59
    minDesiredWidth=30
    minFractionalDigits=0
    minSliderValue=1
    minValue=1
    toolTipText=Minute
    value=59
    foregroundColor {colorUseRule=UseColor_Specified, specifiedColor=<rgba 1 1 1 1>}
   }
   slot {class=HorizontalBoxSlot}
  }
  CommonDateTimeTextBlock_0 {
   widgetClassPath=/Script/CommonUI.CommonDateTimeTextBlock
   bIsVariable=true
   parent=VerticalBox_0
   widget=CommonDateTimeTextBlock_0
   properties {
    bIsScrollingEnabled=false
    text="31 Dec 1, 23:58:45"
    font {
     bForceMonospaced=false
     bMaterialIsStencil=false
     fontMaterial=null
     fontObject=&NotoSans
     letterSpacing=0
     monospacedWidth=1
     size=12
     skewAmount=0
     typefaceFontName=NotoSans-Regular
     outlineSettings {
      bApplyOutlineToDropShadows=false
      bMiteredCorners=false
      bSeparateFillAlpha=false
      outlineMaterial=null
      outlineSize=0
      outlineColor=<rgba 0 0 0 1>
     }
    }
    shadowOffset=<0 0>
   }
   slot {class=VerticalBoxSlot}
  }
 ]
}
