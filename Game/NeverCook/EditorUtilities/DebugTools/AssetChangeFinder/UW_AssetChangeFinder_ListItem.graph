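# List row widget for the Asset Change Finder editor debug tool, showing one asset's label, date and color-coded class.
# HandleDataChanged unpacks generic list item data into a Unix timestamp and package name, then queries the Asset Registry for the class, colored via ClassColorMap.
# The find button broadcasts onFindClicked with the asset name for external handling.
# Overrides tint the edge by selection, and FilterBySearchString matches search text against the label.
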WidgetBlueprint /Game/NeverCook/EditorUtilities/DebugTools/AssetChangeFinder/UW_AssetChangeFinder_ListItem
graphs [
 EventGraph {nodes [
   "On Clicked (btnFind)"!() -> (OutputDelegate, then, widget:User Widget Object Reference, event:Pointer Event Structure) {is=title K2Node_ComponentBoundEvent}
   > Call On Find Clicked!(execute:Exec, self:UW Asset Change Finder List Item Object Reference, Asset:Name=Get CurrentAsset() -> (CurrentAsset:Name=None)) -> (then:Exec) {is=title K2Node_CallDelegate}
   n1 = Event updateAfterDataChange!() -> (OutputDelegate, then, currentListItemData:Object Reference) {is=title K2Node_Event}
   > Parent: updateAfterDataChange!(execute, currentListItemData:Object Reference=@n1.currentListItemData) -> (then) {is=title K2Node_CallParentFunction}
   > HandleDataChanged!(execute, self, Object:Object Reference=@n1.currentListItemData) -> (then)
  ]}
 FilterBySearchString {nodes [
   n1 = FilterBySearchString!() -> (then, SearchString:String) {is=title K2Node_FunctionEntry}
   > Return Node!(execute, ShouldDisplay:bool=Contains(SearchIn=Get label() -> (label:String), Substring=@n1))
  ]}
 HandleDataChanged {nodes [
   n1 = HandleDataChanged!() -> (then, Object:Object Reference) {is=title K2Node_FunctionEntry}
   > n2 = Cast To TA_GenericListItemData!(execute, Object=@n1) -> (then, CastFailed, AsTA Generic List Item Data:TA Generic List Item Data Object Reference) {is=title K2Node_DynamicCast}
   n3 = "Get (a copy)"(Array:Array of Integers=Get integers!(self:TA Generic List Item Data Object Reference=@n2) -> (integers:Array of Integers)) -> (Output:int)
   > SetDateTimeValue!(
    execute
    self:Common Date Time Text Block Object Reference=Get txtDateTime() -> (txtDateTime:Common Date Time Text Block Object Reference)
    InDateTime:Date Time Structure=FromUnixTimestamp!(UnixTime:Integer64="To Integer64 (Integer)"!(InInt:int=@n3) -> (ReturnValue:Integer64=0)) -> (ReturnValue:Date Time Structure)
    bShowAsCountdown:bool=false
    InRefreshDelay:float=1
   ) -> (then)
   > n4 = Set CurrentAsset(CurrentAsset:Name="Get (a copy)"(Array:Array of Names=Get names!(self:TA Generic List Item Data Object Reference=@n2) -> (names:Array of Names), Dimension 1=1) -> (Output:Name)) -> (Output_Get:Name=None)
   n5 = GetAssetRegistry()
   > n6 = GetAssetsByPackageName!(execute, self:Asset Registry Interface=@n5, PackageName:Name=@n4, bIncludeOnlyOnDiskAssets:bool=false, bSkipARFilteredAssets:bool=true) -> (then, OutAssetData:Array of Asset Data Structures, ReturnValue:bool)
   > Branch(Condition=Is Valid Index!(TargetArray:Array of Asset Data Structures=@n6, IndexToTest:int) -> (ReturnValue:bool))
   n7 = Break Asset Data(AssetData="Get (a copy)"(Array:Array of Asset Data Structures=@n6) -> (Output:Asset Data Structure))
   > Set l_className(l_className:String=Split(
     SourceString=BreakTopLevelAssetPath!(TopLevelAssetPath:"Top Level Asset Path Structure (by ref)"=@n7.AssetClassPath) -> (PathString:String)
     InStr=.
     SearchDir=FromEnd
    ).RightS) -> (Output_Get:String)
   > "SetText (Text)"(self=Get txtType() -> (txtType:Text Object Reference), InText="To Text (String)"(InString=Get l_className() -> (l_className:String)))
   n8 = Find(
    TargetMap:Map of Strings to Linear Color Structures=Get ClassColorMap() -> (ClassColorMap:Map of Strings to Linear Color Structures)
    Key=Get l_className()
   ) -> (Value:Linear Color Structure)
   > Branch(Condition=@n8.ReturnValue)
   > SetColorAndOpacity(self=Get txtType(), InColorAndOpacity=Make Slate Color(SpecifiedColor=@n8))
   GetAssets!(execute, self:Asset Registry Interface=@n5, Filter:"ARFilter Structure (by ref)", bSkipARFilteredAssets:bool=true) -> (then, OutAssetData:Array of Asset Data Structures, ReturnValue:bool)
  ]}
 getBackgroundWidget {nodes [
   getBackgroundWidget() {is=title K2Node_FunctionEntry}
   > Return Node!(execute, backgroundWidget:Widget Object Reference=Get borderMain() -> (borderMain:Border Object Reference))
  ]}
 getLabelWidget {nodes [
   getLabelWidget() {is=title K2Node_FunctionEntry}
   > Return Node!(execute, labelWidget:Widget Object Reference=Get txtMain() -> (txtMain:Text Object Reference))
  ]}
 onFindClicked {nodes [onFindClicked!() -> (then, Asset:Name) {is=title K2Node_FunctionEntry}]}
 setTextAlignment {nodes [
   n1 = setTextAlignment!() -> (then, alignment:EHorizontalAlignment Enum=HAlign_Fill) {is=title K2Node_FunctionEntry}
   > n2 = Cast To ScaleBoxSlot!(execute, Object=Get Slot!(self:Widget Object Reference=Get txtMain()) -> (Slot:Panel Slot Object Reference)) -> (then, CastFailed, AsScale Box Slot:Scale Box Slot Object Reference) {is=title K2Node_DynamicCast}
   > SetHorizontalAlignment(self:Scale Box Slot Object Reference=@n2, InHorizontalAlignment=@n1)
  ]}
 validateBackgroundColor {nodes [
   validateBackgroundColor() {is=title K2Node_FunctionEntry}
   > SetColorAndOpacity(
    self:Image Object Reference=Get imageEdge() -> (imageEdge:Image Object Reference)
    InColorAndOpacity:Linear Color Structure=SelectColor(A=(R=0.0284,G=0.3763,B=1,A=1), B=(R=0.8123,G=0.8217,B=0.8385,A=1), bPickA=getSelected!() -> (isSelected:bool))
   )
   > n1 = Parent: validateBackgroundColor!(execute) -> (then, handled:bool) {is=title K2Node_CallParentFunction}
   > Return Node!(execute, handled:bool=@n1)
  ]}
]
variables [
 ClassColorMap {
  type="map<LinearColor>"
  default {
   Blueprint=<rgba 0.1374 0.3882 1 1>
   Material=<rgba 0.0784 0.7882 0.0824 1>
   MaterialFunction=<rgba 0.0349 0.7791 1 1>
   MaterialInstanceConstant=<rgba 0.0575 0.5781 0.0604 1>
   Texture2D=<rgba 0.7 0.015 0 1>
  }
 }
 CurrentAsset {type=string, default=None}
 onFindClicked {}
]
widgets {
 info {
  inheritedWidgetCount=0
  parentClass=/Game/Blueprints/Blutilities/ForTechArt/Widgets/Buttons/TA_BaseEditorButtonWidget.TA_BaseEditorButtonWidget_C
  rootWidgetClass=&Overlay
  widgetCount=12
 }
 widgets [
  Overlay_7 {widgetClassPath=&Overlay, widget=Overlay_7}
  VerticalBox_0 {
   widgetClassPath=&VerticalBox
   parent=Overlay_7
   widget=VerticalBox_0
   slot {class=OverlaySlot, properties {horizontalAlignment=HAlign_Fill}}
  }
  borderMain {
   widgetClassPath=&Border
   bIsVariable=true
   parent=VerticalBox_0
   widget=borderMain
   properties {padding=<ltrb 3 3 3 3>}
   slot {class=VerticalBoxSlot}
  }
  HorizontalBox_0 {
   widgetClassPath=&HorizontalBox
   parent=borderMain
   widget=HorizontalBox_0
   slot {class=BorderSlot, properties {padding=<ltrb 3 3 3 3>}}
  }
  imgIcon {
   widgetClassPath=&Image
   bIsVariable=true
   parent=HorizontalBox_0
   widget=imgIcon
   properties {
    visibility=Collapsed
    brush {
     bIsDynamicallyLoaded=false
     drawAs=Image
     imageType=NoImage
     mirroring=NoMirror
     resourceName=None
     resourceObject=null
     tiling=NoTile
     imageSize=<14 14>
     margin=<ltrb 0 0 0 0>
     outlineSettings {
      bUseBrushTransparency=false
      roundingType=HalfHeightRadius
      width=0
      color {colorUseRule=UseColor_Specified, specifiedColor=<rgba 0 0 0 0>}
      cornerRadii=<0 0 0 0>
     }
     tintColor {colorUseRule=UseColor_Specified, specifiedColor=<rgba 1 1 1 1>}
     uVRegion {bIsValid=false, max=<0 0>, min=<0 0>}
    }
   }
   slot {class=HorizontalBoxSlot, properties {horizontalAlignment=HAlign_Center, verticalAlignment=VAlign_Center}}
  }
  ScaleBox_0 {
   widgetClassPath=&ScaleBox
   parent=HorizontalBox_0
   widget=ScaleBox_0
   properties {clipping=ClipToBounds, stretchDirection=DownOnly}
   slot {class=HorizontalBoxSlot, properties {padding=<ltrb 5 0 0 0>, size {sizeRule=Fill, value=1}}}
  }
  txtMain {
   widgetClassPath=&TextBlock
   bIsVariable=true
   parent=ScaleBox_0
   widget=txtMain
   properties {
    clipping=ClipToBounds
    justification=Center
    text=Text Block
    colorAndOpacity {colorUseRule=UseColor_Specified, specifiedColor=<rgba 0.0104 0.0104 0.0104 1>}
    font {
     bForceMonospaced=false
     bMaterialIsStencil=false
     fontMaterial=null
     fontObject=&Roboto
     letterSpacing=0
     monospacedWidth=1
     size=10.5
     skewAmount=0
     typefaceFontName=Bold
     outlineSettings {
      bApplyOutlineToDropShadows=false
      bMiteredCorners=false
      bSeparateFillAlpha=false
      outlineMaterial=null
      outlineSize=0
      outlineColor=<rgba 0 0 0 1>
     }
    }
    shadowColorAndOpacity=<rgba 0 0 0 0.2>
   }
   slot {class=ScaleBoxSlot, properties {horizontalAlignment=HAlign_Fill}}
  }
  txtDateTime {
   widgetClassPath=/Script/CommonUI.CommonDateTimeTextBlock
   bIsVariable=true
   parent=HorizontalBox_0
   widget=txtDateTime
   properties {
    justification=Center
    minDesiredWidth=110
    text=Text Block
    font {
     bForceMonospaced=false
     bMaterialIsStencil=false
     fontMaterial=null
     fontObject=&NotoSans
     letterSpacing=0
     monospacedWidth=1
     size=10.5
     skewAmount=0
     typefaceFontName=NotoSans-Regular
     outlineSettings {
      bApplyOutlineToDropShadows=false
      bMiteredCorners=false
      bSeparateFillAlpha=false
      outlineMaterial=null
      outlineSize=0
      outlineColor=<rgba 0 0 0 1>
     }
    }
    shadowOffset=<0 0>
   }
   slot {class=HorizontalBoxSlot, properties {verticalAlignment=VAlign_Center, padding=<ltrb 5 0 0 0>}}
  }
  SclaeBoxType {
   widgetClassPath=&ScaleBox
   parent=HorizontalBox_0
   widget=SclaeBoxType
   properties {clipping=ClipToBounds, stretchDirection=DownOnly}
   slot {class=HorizontalBoxSlot, properties {padding=<ltrb 5 0 0 0>, size {sizeRule=Fill, value=1}}}
  }
  txtType {
   widgetClassPath=&TextBlock
   bIsVariable=true
   parent=SclaeBoxType
   widget=txtType
   properties {
    clipping=ClipToBounds
    justification=Center
    font {
     bForceMonospaced=false
     bMaterialIsStencil=false
     fontMaterial=null
     fontObject=&Roboto
     letterSpacing=0
     monospacedWidth=1
     size=10.5
     skewAmount=0
     typefaceFontName=Bold
     outlineSettings {
      bApplyOutlineToDropShadows=false
      bMiteredCorners=false
      bSeparateFillAlpha=false
      outlineMaterial=null
      outlineSize=0
      outlineColor=<rgba 0 0 0 1>
     }
    }
    shadowColorAndOpacity=<rgba 0 0 0 0.2>
   }
   slot {class=ScaleBoxSlot, properties {horizontalAlignment=HAlign_Fill}}
  }
  btnFind {
   widgetClassPath=&TA_Generic01_Button
   parent=HorizontalBox_0
   widget=btnFind
   properties {iconBrush {
     bIsDynamicallyLoaded=false
     drawAs=Image
     imageType=FullColor
     mirroring=NoMirror
     resourceName=None
     resourceObject=/Game/Athena/UI/Challenges/Art/ChallengeTypeIcons/UI_ChallengeIcon_Find
     tiling=NoTile
     imageSize=<14 14>
     margin=<ltrb 0 0 0 0>
     outlineSettings {
      bUseBrushTransparency=false
      roundingType=HalfHeightRadius
      width=0
      color {colorUseRule=UseColor_Specified, specifiedColor=<rgba 0 0 0 0>}
      cornerRadii=<0 0 0 1>
     }
     tintColor {colorUseRule=UseColor_Specified, specifiedColor=<rgba 1 1 1 1>}
     uVRegion {bIsValid=false, max=<0 0>, min=<0 0>}
    }}
   slot {class=HorizontalBoxSlot}
  }
  imageEdge {
   widgetClassPath=&Image
   bIsVariable=true
   parent=VerticalBox_0
   widget=imageEdge
   properties {
    brush {
     bIsDynamicallyLoaded=false
     drawAs=Image
     imageType=NoImage
     mirroring=NoMirror
     resourceName=None
     resourceObject=null
     tiling=NoTile
     imageSize=<32 2>
     margin=<ltrb 0 0 0 0>
     outlineSettings {
      bUseBrushTransparency=false
      roundingType=HalfHeightRadius
      width=0
      color {colorUseRule=UseColor_Specified, specifiedColor=<rgba 0 0 0 0>}
      cornerRadii=<0 0 0 0>
     }
     tintColor {colorUseRule=UseColor_Specified, specifiedColor=<rgba 1 1 1 1>}
     uVRegion {bIsValid=false, max=<0 0>, min=<0 0>}
    }
    colorAndOpacity=<rgba 0.2031 0.2031 0.2031 1>
   }
   slot {class=VerticalBoxSlot}
  }
 ]
}
