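# Editor utility widget that finds assets changed in Perforce between two dates, to help trace which change caused a bug.
# Runs can filter by the selected asset's dependencies, a P4 username or both, listing results with Content Browser find buttons.
# Embedded Python through DebugToolLib checks the P4 connection, lists workspaces and validates usernames.
# Tick tracks selection and filters to update the run button, while a save game persists workspace, user, dates and filters.
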EditorUtilityWidgetBlueprint /Game/NeverCook/EditorUtilities/DebugTools/AssetChangeFinder/EUW_AssetChangeFinder
graphs [
 EventGraph {nodes [
   "On Clicked (btnRunAssetFilter)"!() -> (OutputDelegate, then, widget:User Widget Object Reference, event:Pointer Event Structure) {is=title K2Node_ComponentBoundEvent}
   n1 = GetIsEnabled!(self:Widget Object Reference=Get cbUserFilter() -> (cbUserFilter:TA Generic 01 Checkbox Object Reference)) -> (ReturnValue:bool)
   n2 = GetIsEnabled(self=Get cbAssetFilter() -> (cbAssetFilter:TA Generic 01 Checkbox Object Reference))
   > Branch(Condition=OR Boolean(A=@n1, B=@n2))
   > HandleFindChanges!(execute, self, filterUserName:bool=@n1, filterPackages:bool=@n2) -> (then)
   "On Clicked (btnFind)"!() -> (OutputDelegate, then, widget:User Widget Object Reference, event:Pointer Event Structure) {is=title K2Node_ComponentBoundEvent}
   n3 = Get LastRunPackage() -> (LastRunPackage:Name=None)
   > SyncBrowserToObjects(AssetPaths=Make Array("[0]":String="To String (Name)"(InName=@n3)) -> (Array:Array of Strings))
   "On Clicked (cbUserFilter)"!() -> (OutputDelegate, then, widget:User Widget Object Reference, event:Pointer Event Structure) {is=title K2Node_ComponentBoundEvent}
   > InitUserNameSection()
   > SaveGame_Save(self:TA Base Editor Widget Object Reference)
   "On Clicked (cbAssetFilter)"!() -> (OutputDelegate, then, widget:User Widget Object Reference, event:Pointer Event Structure) {is=title K2Node_ComponentBoundEvent}
   > SaveGame_Save()
   n4 = "On Selection Changed (sComboClients)"!() -> (OutputDelegate, then, SelectedItem:String, SelectionType:ESelectInfo Enum) {is=title K2Node_ComponentBoundEvent}
   > Set SelectedWorkspace(SelectedWorkspace:String=@n4.SelectedItem) -> (Output_Get:String)
   > SaveGame_Save()
   "On Clicked (btnHelp)"!() -> (OutputDelegate, then, widget:User Widget Object Reference, event:Pointer Event Structure) {is=title K2Node_ComponentBoundEvent}
   > Show Message Dialog(
    Title=Format Text(Format="LOCTEXT(\"Asset Change Finder Help\")")
    Message=Format Text(Format="LOCTEXT(\"The Asset Change Finder will search all dependencies of a selected asset and find anything that changed between two dates.  This is mostly for a situation when a bug appears and no-one knows which change caused it.  \")")
    MessageType=Ok
    DefaultValue=Ok
    MessageCategory=Info
   )
   n5 = "On Time Changed (timePickerLatest)"() {is=title K2Node_ComponentBoundEvent}
   SaveGame_Save(execute=[@n5.then, @n6.then])
   > OnTick_SetButtonLabel!(execute, self, ForceEval:bool=true) -> (then)
   n6 = "On Time Changed (timePickerEarliest)"() {is=title K2Node_ComponentBoundEvent}
   "On Clicked (btnAssignUserName)"!() -> (OutputDelegate, then, widget:User Widget Object Reference, event:Pointer Event Structure) {is=title K2Node_ComponentBoundEvent}
   > InitUserNameTextBox!(execute, self, ForceCurrent:bool=true) -> (then)
   > SaveGame_Save()
   n7 = "On Text Committed (edTextUser)"!() -> (OutputDelegate, then, Text:"Text (by ref)", CommitMethod:ETextCommit Enum=Default) {is=title K2Node_ComponentBoundEvent}
   > n8 = Switch on ETextCommit!(execute, Selection:ETextCommit Enum=@n7.CommitMethod) -> (Default:Exec, OnEnter:Exec, OnUserMovedFocus:Exec, OnCleared:Exec) {is=title K2Node_SwitchEnum}
   SaveGame_Save(execute=[@n8.Default, @n8.OnEnter, @n8.OnUserMovedFocus])
   > ValidateUserName()
   Event Construct()
   > Parent: Construct()
   > ValidateUserName()
   > OnTick_SetButtonLabel(ForceEval=true)
   Event Tick()
   > OnTick_SetButtonLabel(ForceEval=false)
  ]}
 GetClients {nodes [
   GetClients() {is=title K2Node_FunctionEntry}
   > n1 = Execute Python Script!(
    execute
    PythonScript="import TechArt.Tools.DebugTools.DebugToolLib as tool_lib\r\nfrom importlib import reload\r\nreload(tool_lib)\r\nclients = tool_lib.getClients()"
   ) -> (then, ReturnValue, clients:Array of Strings) {outputs [clients]}
   > n2 = Branch(Condition=@n1)
   > Return Node!(execute, clients:Array of Strings=@n1.clients)
   Return Node(execute=@n2.else)
  ]}
 HandleFindChanges {nodes [
   HandleFindChanges!() -> (then, filterUserName:bool=~, filterPackages:bool=~) {is=title K2Node_FunctionEntry}
   > n1 = Sequence()
   > Branch(Condition=Get filterPackages() -> (filterPackages:bool))
   n2 = Get CurrentSelectionPackageName() -> (CurrentSelectionPackageName:Name=None)
   > UpdateLastRunPackage!(execute, self, LastRunPackage:Name=@n2) -> (then)
   > Branch(Condition=AND Boolean(A="Not Equal (Name)"(A=@n2) -> (ReturnValue=~), B="Not Equal (Name)"(A=@n2) -> (ReturnValue=~)))
   > n3 = FindDependencies!(
    execute
    PackageName:Name=Get CurrentSelectionPackageName() -> (CurrentSelectionPackageName=None)
    CheckedAssetsIn:Set of Names=Get l_checkedAsset() -> (l_checkedAsset:Set of Names)
    __WorldContext:Object Reference=Get timePickerEarliest() -> (timePickerEarliest:EUW Time Picker Object Reference)
   ) -> (then, CheckedAssetsOut:Array of Names)
   > n4 = Set l_assetsToCheck(l_assetsToCheck:Array of Names=@n3) -> (Output_Get:Array of Names)
   > Add(TargetArray:Array of Names=@n4, NewItem:"Name (by ref)"=Get CurrentSelectionPackageName() -> (CurrentSelectionPackageName=None)) {is=title K2Node_CallArrayFunction}
   n5 = Branch(execute=@n1.then_1, Condition=Get filterUserName() -> (filterUserName:bool))
   n6 = Get timePickerEarliest()
   n7 = Get edTextUser() -> (edTextUser:Editor Utility Editable Text Box Object Reference)
   > n8 = GetUserChangedAssetsInTimeSpan!(
    execute
    user_name:String="To String (Text)"(InText="GetText (Text Box)"!(self:Text Box Object Reference=@n7) -> (ReturnValue:Text))
    DateEarliest:Date Time Structure=getDateTime!(self:EUW Time Picker Object Reference=@n6) -> (DateTime:Date Time Structure)
    DateLatest:Date Time Structure=getDateTime(self=Get timePickerLatest() -> (timePickerLatest:EUW Time Picker Object Reference))
    ClientName:String=Get SelectedWorkspace() -> (SelectedWorkspace:String)
    AssetPackageNamesIn:Array of Names=Get l_assetsToCheck() -> (l_assetsToCheck:Array of Names)
    __WorldContext:Object Reference=@n6
   ) -> (then, ChangedPackageNames:Array of Names, UTCDates:Array of Integers, Changelists:Array of Integers)
   > UpdateListFromChangeData!(execute, self, PackageNames:Array of Names=@n8, UTCDates:Array of Integers=@n8.UTCDates, Changelists:Array of Integers=@n8.Changelists) -> (then)
   n9 = Get timePickerEarliest()
   n10 = FilterAssetsByChangeDate!(
    execute=@n5.else
    AssetPackageNamesIn:Array of Names=Get l_assetsToCheck()
    DateEarliest:Date Time Structure=getDateTime(self=@n9)
    DateLatest:Date Time Structure=getDateTime(self=Get timePickerLatest())
    ClientName:String=Get SelectedWorkspace()
    __WorldContext:Object Reference=@n9
   ) -> (then, ChangedPackageNames:Array of Names, UTCDates:Array of Integers, Changelists:Array of Integers)
   > UpdateListFromChangeData(PackageNames=@n10, UTCDates=@n10.UTCDates, Changelists=@n10.Changelists)
  ]}
 HandleFindChangesNoAssetFilter {nodes [
   n1 = HandleFindChangesNoAssetFilter!() -> (then, AssetPackageNamesIn:Array of Names) {is=title K2Node_FunctionEntry}
   n2 = Get timePickerEarliest()
   n3 = Get edTextUser()
   > n4 = GetUserChangedAssetsInTimeSpan(
    user_name="To String (Text)"(InText="GetText (Text Box)"(self=@n3))
    DateEarliest=getDateTime(self=@n2)
    DateLatest=getDateTime(self=Get timePickerLatest())
    ClientName=Get SelectedWorkspace()
    AssetPackageNamesIn=@n1
    __WorldContext=@n2
   )
   > UpdateListFromChangeData(PackageNames=@n4, UTCDates=@n4.UTCDates, Changelists=@n4.Changelists)
   Get l_assetsToCheck()
  ]}
 HandleListItemFindClicked {nodes [
   n1 = HandleListItemFindClicked!() -> (then, Asset:Name) {is=title K2Node_FunctionEntry}
   > SyncBrowserToObjects(AssetPaths=Make Array("[0]"="To String (Name)"(InName=@n1)))
  ]}
 HandlePipelineConnectionFailed {nodes [
   HandlePipelineConnectionFailed() {is=title K2Node_FunctionEntry}
   > SetVisibility(self=Get panelUnpipeFailed() -> (panelUnpipeFailed:Border Object Reference), InVisibility=SelfHitTestInvisible)
  ]}
 InitAllPostLoad {nodes [
   InitAllPostLoad() {is=title K2Node_FunctionEntry}
   > n1 = Try Init P4!(execute:Exec) -> (Valid:Exec, NoP4:Exec, NoPipe:Exec) {is=title K2Node_MacroInstance}
   HandlePipelineConnectionFailed(execute=[@n1.NoP4, @n1.NoPipe]) {is=title K2Node_CallFunction}
   SetVisibility(execute=@n1.Valid, self=Get panelUnpipeFailed(), InVisibility=Collapsed)
   > InitClientCombobox()
   > InitUserNameTextBox(ForceCurrent=false)
   > InitUserNameSection()
  ]}
 InitClientCombobox {nodes [
   InitClientCombobox() {is=title K2Node_FunctionEntry}
   > n1 = Branch(Condition=IsEmpty(InString=Get SelectedWorkspace()))
   > n2 = Execute Python Script!(
    execute
    PythonScript="import unreal\r\nimport TechArt.Tools.DebugTools.DebugToolLib as tool_lib\r\nfrom importlib import reload\r\nreload(tool_lib)\r\ncurrent_client = tool_lib.getCurrentClient()\r\n"
   ) -> (then, ReturnValue, current_client:String) {outputs [current_client]}
   > n3 = Set SelectedWorkspace(SelectedWorkspace=@n2.current_client)
   > n4 = PrintString(InString=Append(A="\r\n\r\n\r\n\r\nClient ", B=@n3))
   n5 = GetClients!(execute=[@n4.then, @n1.else], self) -> (then, clients:Array of Strings) {is=title K2Node_CallFunction}
   > n6 = For Each Loop(Array=@n5)
   n7 = Get SelectedWorkspace()
   n8 = Get sComboClients() -> (sComboClients:"Searchable Combo Box (String) Object Reference")
   n9 = Get sComboClients()
   SetSelectedIndex!(
    execute=@n6.Completed
    self:"Searchable Combo Box (String) Object Reference"=@n9
    Index:int="Max (Integer)"(A=FindOptionIndex!(self:"Searchable Combo Box (String) Object Reference"=@n8, Option:String=@n7) -> (ReturnValue:int=0), B=0)
   ) -> (then)
   AddOption!(execute=@n6.LoopBody, self:"Searchable Combo Box (String) Object Reference"=@n9, Option:String=@n6) -> (then)
  ]}
 InitUserNameSection {nodes [
   InitUserNameSection() {is=title K2Node_FunctionEntry}
   n1 = getSelected!(self:TA Base Editor Button Widget Object Reference=Get cbUserFilter()) -> (isSelected:bool)
   > SetIsEnabled(self=Get panelUserNameFilter() -> (panelUserNameFilter:Horizontal Box Object Reference), bInIsEnabled=@n1)
   > SetToolTipText!(
    execute
    self:Widget Object Reference=Get panelUserNameFilter()
    InToolTipText:"Text (by ref)"=Select(
     Option 0:"Text (by ref)"=MakeLiteralText(Value="LOCTEXT(\"Enable the User Filter checkbox near the run button to choose a user name to filter by.\")")
     Option 1:"Text (by ref)"
     Index=@n1
    ) -> (ReturnValue:"Text (by ref)")
   ) -> (then)
   > SetVisibility(
    self=Get panelUserNameFilter()
    InVisibility=Select(Option 0:ESlateVisibility Enum=Visible, Option 1:ESlateVisibility Enum=SelfHitTestInvisible, Index=@n1) -> (ReturnValue:ESlateVisibility Enum)
   )
  ]}
 InitUserNameTextBox {nodes [
   n1 = InitUserNameTextBox!() -> (then, ForceCurrent:bool=~) {is=title K2Node_FunctionEntry}
   n2 = "GetText (Text Box)"(self=Get edTextUser())
   > Branch(Condition=OR Boolean(A=@n1, B=TextIsEmpty(InText=@n2)))
   > n3 = Execute Python Script!(
    execute
    PythonScript="import unreal\r\nimport TechArt.Tools.DebugTools.DebugToolLib as tool_lib\r\n\r\np4_connect = tool_lib.getP4Connection()\r\nuser_name = None\r\nif p4_connect is not None:\r\n    infos = p4_connect.run('','info')\r\n    if len(infos) > 0:\r\n        user_name = infos[0]['userName']\r\n"
   ) -> (then, ReturnValue, user_name:String) {outputs [user_name]}
   > "SetText (Text Box)"!(execute, self:Text Box Object Reference=Get edTextUser(), InText:Text="To Text (String)"(InString=@n3.user_name)) -> (then)
  ]}
 OnTick_SetButtonLabel {nodes [
   OnTick_SetButtonLabel!() -> (then, ForceEval:bool=~) {is=title K2Node_FunctionEntry}
   > n1 = Sequence!(execute) -> (then_0, then_1, then_2:Exec, then_3:Exec)
   > n2 = GetSelectedAssetData()
   > n3 = Branch(Condition=Is Valid Index!(TargetArray:Array of Asset Data Structures=@n2, IndexToTest:int) -> (ReturnValue:bool))
   n4 = Break Asset Data(AssetData="Get (a copy)"(Array:Array of Asset Data Structures=@n2) -> (Output:Asset Data Structure))
   > Set l_newSelection(l_newSelection:Name=@n4) -> (Output_Get:Name=None)
   > Set l_assetName(l_assetName:Name=@n4.AssetName) -> (Output_Get:Name=None)
   Set l_newSelection(execute=@n3.else, l_newSelection=None) -> (Output_Get=None)
   n5 = Get l_newSelection() -> (l_newSelection:Name=None)
   n6 = Get CurrentSelectionPackageName() -> (CurrentSelectionPackageName=None)
   n7 = Get lastUserNameValid() -> (lastUserNameValid:bool)
   n8 = Get UserNameValid() -> (UserNameValid:bool)
   n9 = getSelected(self=Get cbAssetFilter())
   n10 = Get lastAssetFilter() -> (lastAssetFilter:bool)
   n11 = getSelected(self=Get cbUserFilter())
   n12 = Get lastUserFilter() -> (lastUserFilter:bool)
   Set l_allowEval(
    execute=@n1.then_1
    l_allowEval:bool=OR Boolean!(
     A="Not Equal (Name)"(A=@n6, B=@n5) -> (ReturnValue=~)
     B="Not Equal (Boolean)"(A=@n10, B=@n9) -> (ReturnValue=~)
     C:bool="Not Equal (Boolean)"(A=@n12, B=@n11) -> (ReturnValue=~)
     D:bool=Get ForceEval() -> (ForceEval:bool)
     E:bool="Not Equal (Boolean)"(A=@n7, B=@n8) -> (ReturnValue=~)
    ) -> (ReturnValue)
   ) -> (Output_Get:bool)
   > Set CurrentSelectionPackageName(CurrentSelectionPackageName:Name=Get l_newSelection() -> (l_newSelection=None)) -> (Output_Get:Name=None)
   > Set lastAssetFilter(lastAssetFilter:bool=getSelected(self=Get cbAssetFilter())) -> (Output_Get:bool)
   > Set lastUserFilter(lastUserFilter:bool=getSelected(self=Get cbUserFilter())) -> (Output_Get:bool)
   > Set lastUserNameValid(lastUserNameValid:bool=Get UserNameValid()) -> (Output_Get:bool)
   Set l_validAssetSelected(execute=@n1.then_2, l_validAssetSelected:bool="Not Equal (Name)"(A=Get l_newSelection() -> (l_newSelection=None)) -> (ReturnValue=~)) -> (Output_Get:bool)
   Branch(execute=@n1.then_3, Condition=Get l_allowEval() -> (l_allowEval:bool))
   > n13 = Sequence!(execute) -> (then_0, then_1, then_2)
   K2Node_Composite_0 = GetButtonLabel!() -> (ReturnValue:String) {is=title K2Node_Composite}
   > setLabel!(
    execute
    self:TA Base Editor Button Widget Object Reference=Get btnRunAssetFilter() -> (btnRunAssetFilter:TA Generic 01 Button Object Reference)
    label:String=@K2Node_Composite_0
   ) -> (then)
   K2Node_Composite_1 = GetButtonTooltip!() -> (ReturnValue:"Text (by ref)") {is=title K2Node_Composite}
   SetToolTipText(execute=@n13.then_1, self=Get btnRunAssetFilter(), InToolTipText=@K2Node_Composite_1)
   n14 = Get lastAssetFilter()
   n15 = Get lastUserFilter()
   n16 = Get lastUserFilter()
   n17 = Get lastUserNameValid()
   n18 = Get l_validAssetSelected() -> (l_validAssetSelected:bool)
   n19 = Get lastAssetFilter()
   SetIsEnabled(
    execute=@n13.then_2
    self=Get btnRunAssetFilter()
    bInIsEnabled=AND Boolean!(
     A=OR Boolean!(A=@n14, B=@n15) -> (ReturnValue)
     B=Select(Option 0:bool=true, Option 1:bool=@n18, Index=@n19) -> (ReturnValue:bool)
     C:bool=Select(Option 0=true, Option 1=@n17, Index=@n16)
    ) -> (ReturnValue)
   )
   n20 = Get btnRunAssetFilter()
   n21 = Get btnRunAssetFilter()
   > Set buttonColor_Default!(
    execute
    buttonColor_Default:Linear Color Structure=Select(
     Option 0:Linear Color Structure=(R=0.2604,G=0.0245,B=0,A=1)
     Option 1:Linear Color Structure=(R=0.0227,G=0.5573,B=0.0617,A=1)
     Index=GetIsEnabled(self=@n20)
    ) -> (ReturnValue:Linear Color Structure)
    self:TA Base Editor Button Widget Object Reference=@n21
   ) -> (then, Output_Get:Linear Color Structure)
   > validateVisuals!(execute, self:TA Base Editor Button Widget Object Reference=@n21) -> (then, " ":bool)
  ]}
 GetButtonLabel {
  path=OnTick_SetButtonLabel.K2Node_Composite_0.GetButtonLabel
  nodes [
   Inputs!()
   n1 = OR Boolean(A=Get lastUserFilter(), B=Get lastAssetFilter())
   n2 = Format Text(Format="LOCTEXT(\"Must have at least one filter enabled.\")")
   n3 = Format Text!(Format="LOCTEXT(\"Find changes related to {asset}\")", asset:Name=Get l_assetName() -> (l_assetName:Name=None)) -> (Result)
   n4 = Get edTextUser()
   n5 = Get edTextUser()
   n6 = Select(
    Option 0:"Text (by ref)"=Format Text!(Format="LOCTEXT(\"User Name filter is enabled but {user} is not a valid P4 username.\")", user:Text="GetText (Text Box)"(self=@n5)) -> (Result)
    Option 1:"Text (by ref)"=Format Text!(
     Format="LOCTEXT(\"Find changes related to {asset} made by {user}\")"
     asset:Name=Get l_assetName() -> (l_assetName=None)
     user="GetText (Text Box)"(self=@n4)
    ) -> (Result)
    Index=Get lastUserNameValid()
   ) -> (ReturnValue:"Text (by ref)")
   n7 = Get lastUserFilter()
   n8 = "GetText (Text Box)"(self=Get edTextUser())
   n9 = "GetText (Text Box)"(self=Get edTextUser())
   n10 = Select(
    Option 0=Select(
     Option 0=Format Text!(Format="LOCTEXT(\"User Name filter is enabled but {user} is not a valid P4 username.\")", user=@n8) -> (Result)
     Option 1=Format Text(Format="LOCTEXT(\"Find changes made by {user}\")", user=@n9)
     Index=Get lastUserNameValid()
    )
    Option 1=Select(
     Option 0=Format Text!(Format="LOCTEXT(\"Asset filter is enabled but no asset is selected.\")") -> (Result)
     Option 1=Select(Option 0=@n3, Option 1=@n6, Index=@n7)
     Index=Get l_validAssetSelected()
    )
    Index=Get lastAssetFilter()
   )
   Outputs!(ReturnValue:String="To String (Text)"(InText=Select(Option 0=@n2, Option 1=@n10, Index=@n1)))
  ]
 }
 GetButtonTooltip {
  path=OnTick_SetButtonLabel.K2Node_Composite_1.GetButtonTooltip
  nodes [
   Inputs()
   n1 = Get lastAssetFilter()
   n2 = Get lastUserFilter()
   n3 = Get edTextUser()
   n4 = Get edTextUser()
   n5 = getDateTime(self=Get timePickerEarliest())
   n6 = getDateTime(self=Get timePickerLatest())
   n7 = Select(
    Option 0=Format Text!(
     Format="LOCTEXT(\"User Name filter is enabled but {user} is not a valid P4 username.  Check your spelling and that P4 is connected.\")"
     user:Text="GetText (Text Box)"(self=@n3)
    ) -> (Result)
    Option 1=Format Text!(
     Format="LOCTEXT(\"Find changes made by {user} between the {date0} and {date1}\")"
     user="GetText (Text Box)"(self=@n4)
     date0:Text=As Date!(InDateTime:"Date Time Structure (by ref)"=@n5, InDateStyle:EDateTimeStyle Enum=Default) -> (ReturnValue:Text)
     date1:Text=As Date(InDateTime=@n6, InDateStyle=Default)
    ) -> (Result)
    Index=Get lastUserNameValid()
   )
   n8 = As Date(InDateTime=getDateTime(self=Get timePickerLatest()), InDateStyle=Default)
   n9 = As Date(InDateTime=getDateTime(self=Get timePickerEarliest()), InDateStyle=Default)
   n10 = Get l_assetName() -> (l_assetName=None)
   n11 = Get timePickerEarliest()
   n12 = Get timePickerLatest()
   n13 = Format Text!(
    Format="LOCTEXT(\"Find changes related to {asset} made by {user} between the {date0} and {date1}\")"
    asset:Name=Get l_assetName() -> (l_assetName=None)
    user="GetText (Text Box)"(self=Get edTextUser())
    date0=As Date(InDateTime=getDateTime(self=@n11), InDateStyle=Default)
    date1=As Date(InDateTime=getDateTime(self=@n12), InDateStyle=Default)
   ) -> (Result)
   n14 = Format Text!(
    Format="LOCTEXT(\"User Name filter is enabled but {user} is not a valid P4 username.  Check your spelling and that P4 is connected.\")"
    user="GetText (Text Box)"(self=Get edTextUser())
   ) -> (Result)
   n15 = Get lastUserNameValid()
   n16 = Select(
    Option 0=Format Text!(Format="LOCTEXT(\"While the Asset Filter is enabled you must select an asset to run on.\")") -> (Result)
    Option 1=Select(
     Option 0=Format Text!(Format="LOCTEXT(\"Find changes related to {asset} between the {date0} and {date1}\")", asset:Name=@n10, date0:Text=@n9, date1:Text=@n8) -> (Result)
     Option 1=Select(Option 0=@n14, Option 1=@n13, Index=@n15)
     Index=Get lastUserFilter()
    )
    Index=Get l_validAssetSelected()
   )
   n17 = Get lastAssetFilter()
   Outputs(ReturnValue:"Text (by ref)"=Select(
     Option 0=Format Text!(Format="LOCTEXT(\"Without a filter the tool would run on every file in the project which is pretty much impossible.  Select at least one filter to the right to get started.\")") -> (Result)
     Option 1=Select(Option 0=@n7, Option 1=@n16, Index=@n17)
     Index=OR Boolean(A=@n2, B=@n1)
    ))
  ]
 }
 SaveGame_PostLoad {nodes [
   n1 = SaveGame_PostLoad!() -> (then, saveGameExisted:bool, saveGame_Object:Save Game Object Reference) {is=title K2Node_FunctionEntry}
   > n2 = Sequence!(execute) -> (then_0, then_1)
   > Branch(Condition=@n1)
   > n3 = Sequence!(execute) -> (then_0, then_1, then_2:Exec)
   > n4 = Cast To TA_Base_SaveGame!(execute, Object=Get saveGame_Object() -> (saveGame_Object:Save Game Object Reference)) -> (then, CastFailed, AsTA Base Save Game:TA Base Save Game Object Reference) {is=title K2Node_DynamicCast}
   > Set l_save(l_save:TA Base Save Game Object Reference=@n4) -> (Output_Get:TA Base Save Game Object Reference)
   n5 = Get strings!(self:TA Base Save Game Object Reference=@n4) -> (strings:Array of Strings)
   > Set SelectedWorkspace(SelectedWorkspace="Get (a ref)"(Array:Array of Strings=@n5, Dimension 1=0) -> (Output:"String (by ref)"))
   > "SetText (Text Box)"(self=Get edTextUser(), InText="To Text (String)"(InString="Get (a ref)"(Array=@n5, Dimension 1=1)))
   n6 = Get integers!(self:TA Base Save Game Object Reference=Get l_save() -> (l_save:TA Base Save Game Object Reference)) -> (integers:Array of Integers)
   n7 = "Get (a copy)"(Array:Array of Integers=@n6) -> (Output:int)
   SetDateTime!(
    execute=@n3.then_1
    self:EUW Time Picker Object Reference=Get timePickerEarliest()
    InDateTime:Date Time Structure=FromUnixTimestamp!(UnixTime:Integer64="To Integer64 (Integer)"!(InInt:int=@n7) -> (ReturnValue:Integer64=0)) -> (ReturnValue:Date Time Structure)
    AllowEvent:bool=true
   ) -> (then)
   n8 = "Get (a copy)"(Array=@n6, Dimension 1=1)
   > SetDateTime(self=Get timePickerLatest(), InDateTime=FromUnixTimestamp(UnixTime="To Integer64 (Integer)"(InInt=@n8) -> (ReturnValue=0)), AllowEvent=true)
   n9 = Get booleans!(self:TA Base Save Game Object Reference=Get l_save()) -> (booleans:Array of Booleans)
   setSelected!(
    execute=@n3.then_2
    self:TA Base Editor Button Widget Object Reference=Get cbAssetFilter()
    selected:bool="Get (a copy)"(Array:Array of Booleans=@n9) -> (Output:bool)
    throwEvent:bool=false
   ) -> (then)
   > setSelected(self=Get cbUserFilter(), selected="Get (a copy)"(Array=@n9, Dimension 1=1), throwEvent=false)
   InitAllPostLoad(execute=@n2.then_1) {is=title K2Node_CallFunction}
   Return Node!(execute, " ":bool=false)
  ]}
 SaveGame_PrepareToSave {nodes [
   n1 = SaveGame_PrepareToSave!() -> (then, SaveGameObject:Save Game Object Reference) {is=title K2Node_FunctionEntry}
   > n2 = Cast To TA_Base_SaveGame(Object=@n1)
   n3 = Get sComboClients()
   n4 = "GetText (Text Box)"(self=Get edTextUser())
   > Set strings!(
    execute
    strings:Array of Strings=Make Array!(
     "[0]"=GetSelectedOption!(self:"Searchable Combo Box (String) Object Reference"=@n3) -> (ReturnValue:String)
     "[1]":String="To String (Text)"(InText=@n4)
    ) -> (Array)
    self:TA Base Save Game Object Reference=@n2
   ) -> (then, Output_Get:Array of Strings)
   n5 = ToUnixTimestamp!(Time:"Date Time Structure (by ref)"=getDateTime(self=Get timePickerLatest())) -> (ReturnValue:Integer64=0)
   n6 = ToUnixTimestamp(Time=getDateTime(self=Get timePickerEarliest())) -> (ReturnValue=0)
   > Set integers!(
    execute
    integers:Array of Integers=Make Array(
     "[0]":int="To Integer (Integer64)"!(InInt:Integer64=@n6) -> (ReturnValue:int=0)
     "[1]":int="To Integer (Integer64)"(InInt=@n5) -> (ReturnValue=0)
    ) -> (Array:Array of Integers)
    self:TA Base Save Game Object Reference=@n2
   ) -> (then, Output_Get:Array of Integers)
   n7 = Get cbUserFilter()
   n8 = Get cbAssetFilter()
   > Set booleans!(
    execute
    booleans:Array of Booleans=Make Array("[0]":bool=getSelected(self=@n8), "[1]":bool=getSelected(self=@n7)) -> (Array:Array of Booleans)
    self:TA Base Save Game Object Reference=@n2
   ) -> (then, Output_Get:Array of Booleans)
   Return Node(" "=false)
  ]}
 TryInitP4 {nodes [
   Inputs!() -> (execute:Exec)
   > n1 = Execute Python Script!(
    execute
    PythonScript="import TechArt.Tools.DebugTools.DebugToolLib as tool_lib\r\nfrom importlib import reload\r\n\r\nis_valid = tool_lib.getP4Connection() is not None"
   ) -> (then, ReturnValue, is_valid:bool=~) {outputs [is_valid]}
   > n2 = Branch(Condition=@n1)
   > n3 = Branch(Condition=@n1.is_valid)
   > Outputs!(Valid:Exec, NoP4:Exec=@n3.else, NoPipe:Exec=@n2.else)
  ]}
 UpdateLastRunPackage {nodes [
   n1 = UpdateLastRunPackage!() -> (then, LastRunPackage:Name) {is=title K2Node_FunctionEntry}
   > n2 = Set LastRunPackage(LastRunPackage:Name=@n1) -> (Output_Get:Name=None)
   > "SetText (Text)"(
    self=Get CommonTextBlock() -> (CommonTextBlock:Common Text Object Reference)
    InText=Format Text!(Format="LOCTEXT(\"Last Run on {asset}\")", asset:Name=@n2) -> (Result)
   )
   > SetVisibility(
    self=Get hBoxLastrun() -> (hBoxLastrun:Horizontal Box Object Reference)
    InVisibility=Select(Option 0:ESlateVisibility Enum=Hidden, Option 1:ESlateVisibility Enum=Visible, Index="Not Equal (Name)"(A=@n2) -> (ReturnValue=~)) -> (ReturnValue:ESlateVisibility Enum)
   )
  ]}
 UpdateListFromChangeData {nodes [
   n1 = UpdateListFromChangeData!() -> (then, PackageNames:Array of Names, UTCDates:Array of Integers, Changelists:Array of Integers) {is=title K2Node_FunctionEntry}
   > clearItems(self:TA Generic 01 List Object Reference=Get TA_Generic01_List() -> (TA_Generic01_List:TA Generic 01 List Object Reference))
   > n2 = For Each Loop(Array:Array of Names=@n1) -> (Array Element:Name)
   > n3 = Construct Object from Class(Class=/Game/Blueprints/Blutilities/ForTechArt/Data/TA_GenericListItemData.TA_GenericListItemData_C) -> (ReturnValue:TA Generic List Item Data Object Reference)
   n4 = "To String (Name)"(InName=@n2)
   > Set displayString!(
    execute
    displayString:String=Split(SourceString=Split(SourceString=@n4, InStr=., SearchDir=FromEnd), InStr="\\", SearchDir=FromEnd).RightS
    self:TA Generic List Item Data Object Reference=@n3
   ) -> (then, Output_Get:String)
   > n5 = FilenameToPackageName!(execute, Filename:String=@n4) -> (then, ReturnValue:String)
   > Set names!(
    execute
    names:Array of Names=Make Array("[0]":Name=@n2, "[1]":Name=String To Name(InString=@n5)) -> (Array:Array of Names)
    self:TA Generic List Item Data Object Reference=@n3
   ) -> (then, Output_Get:Array of Names)
   > Set integers(
    integers=Make Array(
     "[0]":int="Get (a ref)"(Array:Array of Integers=@n1.UTCDates, Dimension 1=@n2.Array Index) -> (Output:"Integer (by ref)")
     "[1]":int="Get (a ref)"(Array=@n1.Changelists, Dimension 1=@n2.Array Index)
    ) -> (Array:Array of Integers)
    self:TA Generic List Item Data Object Reference=@n3
   )
   > n6 = addItem!(execute, self:TA Generic 01 List Object Reference=Get TA_Generic01_List(), dataObject:Object Reference=@n3) -> (then, newItem:User Widget Object Reference, newData:Object Reference)
   > n7 = Cast To UW_AssetChangeFinder_ListItem!(execute, Object=@n6) -> (then, CastFailed, AsUW Asset Change Finder List Item:UW Asset Change Finder List Item Object Reference) {is=title K2Node_DynamicCast}
   > Bind Event to On Find Clicked(self:UW Asset Change Finder List Item Object Reference=@n7, Delegate=Create Event()) {is=title K2Node_AddDelegate}
  ]}
 ValidateUserName {nodes [
   ValidateUserName() {is=title K2Node_FunctionEntry}
   n1 = Get edTextUser()
   > n2 = Execute Python Script!(
    execute
    PythonScript="import TechArt.Tools.DebugTools.DebugToolLib as tool_lib\r\n\r\ntry:\r\n    p4_connect = tool_lib.getP4Connection()\r\n    users = p4_connect.run(\"\", 'users', in_user)\r\n    valid = len(users) > 0\r\nexcept:\r\n    valid = False"
    in_user:String="To String (Text)"(InText="GetText (Text Box)"(self=@n1))
   ) -> (then, ReturnValue, valid:bool=~) {inputs [in_user], outputs [valid]}
   > Set UserNameValid(UserNameValid:bool=@n2.valid) -> (Output_Get:bool)
  ]}
]
variables [
 CurrentSelectionPackageName {type=string, default=None}
 lastAssetFilter {type=boolean}
 LastRunPackage {type=string, default=None}
 lastUserFilter {type=boolean}
 lastUserNameValid {type=boolean}
 SelectedWorkspace {type=string, default=""}
 UserNameValid {type=boolean}
]
widgets {
 info {
  parentClass=/Game/Blueprints/Blutilities/ForTechArt/Widgets/TA_Base_EditorWidget.TA_Base_EditorWidget_C
  rootWidgetClass=&Overlay
  widgetCount=33
 }
 widgets [
  Overlay_0 {widgetClassPath=&Overlay, widget=Overlay_0}
  StackBox_0 {
   widgetClassPath=&StackBox
   parent=Overlay_0
   widget=StackBox_0
   properties {orientation=Orient_Vertical}
   slot {class=OverlaySlot, properties {horizontalAlignment=HAlign_Fill, verticalAlignment=VAlign_Fill}}
  }
  panelUserNameFilter {
   widgetClassPath=&HorizontalBox
   bIsVariable=true
   parent=StackBox_0
   widget=panelUserNameFilter
   slot {class=StackBoxSlot, properties {padding=<ltrb 0 5 0 5>}}
  }
  CommonTextBlock_1 {
   widgetClassPath=&CommonTextBlock
   parent=panelUserNameFilter
   widget=CommonTextBlock_1
   properties {
    justification=Right
    minDesiredWidth=125
    text="User Name Filter: "
    font {
     bForceMonospaced=false
     bMaterialIsStencil=false
     fontMaterial=null
     fontObject=&NotoSans
     letterSpacing=0
     monospacedWidth=1
     size=10.5
     skewAmount=0
     typefaceFontName=NotoSans-Regular
     outlineSettings {
      bApplyOutlineToDropShadows=false
      bMiteredCorners=false
      bSeparateFillAlpha=false
      outlineMaterial=null
      outlineSize=0
      outlineColor=<rgba 0 0 0 1>
     }
    }
    shadowOffset=<0 0>
   }
   slot {class=HorizontalBoxSlot, properties {verticalAlignment=VAlign_Center, padding=<ltrb 10 0 5 0>}}
  }
  edTextUser {
   widgetClassPath=&EditorUtilityEditableTextBox
   bIsVariable=true
   parent=panelUserNameFilter
   widget=edTextUser
   properties {widgetStyle {
     backgroundColor {colorUseRule=UseColor_Specified, specifiedColor=<rgba 1 1 1 1>}
     backgroundImageFocused {
      bIsDynamicallyLoaded=false
      drawAs=RoundedBox
      imageType=NoImage
      mirroring=NoMirror
      resourceName=None
      resourceObject=null
      tiling=NoTile
      imageSize=<0 0>
      margin=<ltrb 0 0 0 0>
      outlineSettings {
       bUseBrushTransparency=false
       roundingType=FixedRadius
       width=1
       color {colorUseRule=UseColor_ColorTable, specifiedColor=<rgba 0 0 0 0>}
       cornerRadii=<4 4 4 4>
      }
      tintColor {colorUseRule=UseColor_Specified, specifiedColor=<rgba 0.0048 0.0048 0.0048 1>}
      uVRegion {bIsValid=false, max=<0 0>, min=<0 0>}
     }
     backgroundImageHovered {
      bIsDynamicallyLoaded=false
      drawAs=RoundedBox
      imageType=NoImage
      mirroring=NoMirror
      resourceName=None
      resourceObject=null
      tiling=NoTile
      imageSize=<0 0>
      margin=<ltrb 0 0 0 0>
      outlineSettings {
       bUseBrushTransparency=false
       roundingType=FixedRadius
       width=1
       color {colorUseRule=UseColor_ColorTable, specifiedColor=<rgba 0 0 0 0>}
       cornerRadii=<4 4 4 4>
      }
      tintColor {colorUseRule=UseColor_Specified, specifiedColor=<rgba 0.0048 0.0048 0.0048 1>}
      uVRegion {bIsValid=false, max=<0 0>, min=<0 0>}
     }
     backgroundImageNormal {
      bIsDynamicallyLoaded=false
      drawAs=RoundedBox
      imageType=NoImage
      mirroring=NoMirror
      resourceName=None
      resourceObject=null
      tiling=NoTile
      imageSize=<0 0>
      margin=<ltrb 0 0 0 0>
      outlineSettings {
       bUseBrushTransparency=false
       roundingType=FixedRadius
       width=1
       color {colorUseRule=UseColor_ColorTable, specifiedColor=<rgba 0 0 0 0>}
       cornerRadii=<4 4 4 4>
      }
      tintColor {colorUseRule=UseColor_Specified, specifiedColor=<rgba 0.0048 0.0048 0.0048 1>}
      uVRegion {bIsValid=false, max=<0 0>, min=<0 0>}
     }
     backgroundImageReadOnly {
      bIsDynamicallyLoaded=false
      drawAs=RoundedBox
      imageType=NoImage
      mirroring=NoMirror
      resourceName=None
      resourceObject=null
      tiling=NoTile
      imageSize=<0 0>
      margin=<ltrb 0 0 0 0>
      outlineSettings {
       bUseBrushTransparency=false
       roundingType=FixedRadius
       width=1
       color {colorUseRule=UseColor_ColorTable, specifiedColor=<rgba 0 0 0 0>}
       cornerRadii=<4 4 4 4>
      }
      tintColor {colorUseRule=UseColor_Specified, specifiedColor=<rgba 0.0284 0.0284 0.0284 1>}
      uVRegion {bIsValid=false, max=<0 0>, min=<0 0>}
     }
     focusedForegroundColor {colorUseRule=UseColor_Specified, specifiedColor=<rgba 1 1 1 1>}
     foregroundColor {colorUseRule=UseColor_Specified, specifiedColor=<rgba 0.5271 0.5271 0.5271 1>}
     hScrollBarPadding=<ltrb 0 0 0 0>
     padding=<ltrb 8 3 8 4>
     readOnlyForegroundColor {colorUseRule=UseColor_Foreground, specifiedColor=<rgba 1 0 1 1>}
     scrollBarStyle {
      thickness=8
      draggedThumbImage {
       bIsDynamicallyLoaded=false
       drawAs=RoundedBox
       imageType=NoImage
       mirroring=NoMirror
       resourceName=None
       resourceObject=null
       tiling=NoTile
       imageSize=<0 0>
       margin=<ltrb 0 0 0 0>
       outlineSettings {
        bUseBrushTransparency=false
        roundingType=FixedRadius
        width=0
        color {colorUseRule=UseColor_Specified, specifiedColor=<rgba 0 0 0 0>}
        cornerRadii=<4 4 4 4>
       }
       tintColor {colorUseRule=UseColor_Specified, specifiedColor=<rgba 0.2159 0.2159 0.2159 1>}
       uVRegion {bIsValid=false, max=<0 0>, min=<0 0>}
      }
      horizontalBackgroundImage {
       bIsDynamicallyLoaded=false
       drawAs=NoDrawType
       imageType=NoImage
       mirroring=NoMirror
       resourceName=None
       resourceObject=null
       tiling=NoTile
       imageSize=<0 0>
       margin=<ltrb 0 0 0 0>
       outlineSettings {
        bUseBrushTransparency=false
        roundingType=HalfHeightRadius
        width=0
        color {colorUseRule=UseColor_Specified, specifiedColor=<rgba 0 0 0 0>}
        cornerRadii=<0 0 0 0>
       }
       tintColor {colorUseRule=UseColor_Specified, specifiedColor=<rgba 1 1 1 1>}
       uVRegion {bIsValid=false, max=<0 0>, min=<0 0>}
      }
      horizontalBottomSlotImage {
       bIsDynamicallyLoaded=false
       drawAs=NoDrawType
       imageType=NoImage
       mirroring=NoMirror
       resourceName=None
       resourceObject=null
       tiling=NoTile
       imageSize=<0 0>
       margin=<ltrb 0 0 0 0>
       outlineSettings {
        bUseBrushTransparency=false
        roundingType=HalfHeightRadius
        width=0
        color {colorUseRule=UseColor_Specified, specifiedColor=<rgba 0 0 0 0>}
        cornerRadii=<0 0 0 0>
       }
       tintColor {colorUseRule=UseColor_Specified, specifiedColor=<rgba 1 1 1 1>}
       uVRegion {bIsValid=false, max=<0 0>, min=<0 0>}
      }
      horizontalTopSlotImage {
       bIsDynamicallyLoaded=false
       drawAs=NoDrawType
       imageType=NoImage
       mirroring=NoMirror
       resourceName=None
       resourceObject=null
       tiling=NoTile
       imageSize=<0 0>
       margin=<ltrb 0 0 0 0>
       outlineSettings {
        bUseBrushTransparency=false
        roundingType=HalfHeightRadius
        width=0
        color {colorUseRule=UseColor_Specified, specifiedColor=<rgba 0 0 0 0>}
        cornerRadii=<0 0 0 0>
       }
       tintColor {colorUseRule=UseColor_Specified, specifiedColor=<rgba 1 1 1 1>}
       uVRegion {bIsValid=false, max=<0 0>, min=<0 0>}
      }
      hoveredThumbImage {
       bIsDynamicallyLoaded=false
       drawAs=RoundedBox
       imageType=NoImage
       mirroring=NoMirror
       resourceName=None
       resourceObject=null
       tiling=NoTile
       imageSize=<0 0>
       margin=<ltrb 0 0 0 0>
       outlineSettings {
        bUseBrushTransparency=false
        roundingType=FixedRadius
        width=0
        color {colorUseRule=UseColor_Specified, specifiedColor=<rgba 0 0 0 0>}
        cornerRadii=<4 4 4 4>
       }
       tintColor {colorUseRule=UseColor_Specified, specifiedColor=<rgba 0.2159 0.2159 0.2159 1>}
       uVRegion {bIsValid=false, max=<0 0>, min=<0 0>}
      }
      normalThumbImage {
       bIsDynamicallyLoaded=false
       drawAs=RoundedBox
       imageType=NoImage
       mirroring=NoMirror
       resourceName=None
       resourceObject=null
       tiling=NoTile
       imageSize=<0 0>
       margin=<ltrb 0 0 0 0>
       outlineSettings {
        bUseBrushTransparency=false
        roundingType=FixedRadius
        width=0
        color {colorUseRule=UseColor_Specified, specifiedColor=<rgba 0 0 0 0>}
        cornerRadii=<4 4 4 4>
       }
       tintColor {colorUseRule=UseColor_Specified, specifiedColor=<rgba 0.0953 0.0953 0.0953 1>}
       uVRegion {bIsValid=false, max=<0 0>, min=<0 0>}
      }
      verticalBackgroundImage {
       bIsDynamicallyLoaded=false
       drawAs=NoDrawType
       imageType=NoImage
       mirroring=NoMirror
       resourceName=None
       resourceObject=null
       tiling=NoTile
       imageSize=<0 0>
       margin=<ltrb 0 0 0 0>
       outlineSettings {
        bUseBrushTransparency=false
        roundingType=HalfHeightRadius
        width=0
        color {colorUseRule=UseColor_Specified, specifiedColor=<rgba 0 0 0 0>}
        cornerRadii=<0 0 0 0>
       }
       tintColor {colorUseRule=UseColor_Specified, specifiedColor=<rgba 1 1 1 1>}
       uVRegion {bIsValid=false, max=<0 0>, min=<0 0>}
      }
      verticalBottomSlotImage {
       bIsDynamicallyLoaded=false
       drawAs=NoDrawType
       imageType=NoImage
       mirroring=NoMirror
       resourceName=None
       resourceObject=null
       tiling=NoTile
       imageSize=<0 0>
       margin=<ltrb 0 0 0 0>
       outlineSettings {
        bUseBrushTransparency=false
        roundingType=HalfHeightRadius
        width=0
        color {colorUseRule=UseColor_Specified, specifiedColor=<rgba 0 0 0 0>}
        cornerRadii=<0 0 0 0>
       }
       tintColor {colorUseRule=UseColor_Specified, specifiedColor=<rgba 1 1 1 1>}
       uVRegion {bIsValid=false, max=<0 0>, min=<0 0>}
      }
      verticalTopSlotImage {
       bIsDynamicallyLoaded=false
       drawAs=NoDrawType
       imageType=NoImage
       mirroring=NoMirror
       resourceName=None
       resourceObject=null
       tiling=NoTile
       imageSize=<0 0>
       margin=<ltrb 0 0 0 0>
       outlineSettings {
        bUseBrushTransparency=false
        roundingType=HalfHeightRadius
        width=0
        color {colorUseRule=UseColor_Specified, specifiedColor=<rgba 0 0 0 0>}
        cornerRadii=<0 0 0 0>
       }
       tintColor {colorUseRule=UseColor_Specified, specifiedColor=<rgba 1 1 1 1>}
       uVRegion {bIsValid=false, max=<0 0>, min=<0 0>}
      }
     }
     textStyle {
      overflowPolicy=Clip
      transformPolicy=None
      colorAndOpacity {colorUseRule=UseColor_Foreground, specifiedColor=<rgba 1 0 1 1>}
      font {
       bForceMonospaced=false
       bMaterialIsStencil=false
       fontMaterial=null
       fontObject=null
       letterSpacing=0
       monospacedWidth=1
       size=10.5
       skewAmount=0
       typefaceFontName=Regular
       outlineSettings {
        bApplyOutlineToDropShadows=false
        bMiteredCorners=false
        bSeparateFillAlpha=false
        outlineMaterial=null
        outlineSize=0
        outlineColor=<rgba 0 0 0 1>
       }
      }
      highlightColor {colorUseRule=UseColor_ColorTable, specifiedColor=<rgba 0 0 0 0>}
      highlightShape {
       bIsDynamicallyLoaded=false
       drawAs=Image
       imageType=NoImage
       mirroring=NoMirror
       resourceName=None
       resourceObject=null
       tiling=NoTile
       imageSize=<0 0>
       margin=<ltrb 0 0 0 0>
       outlineSettings {
        bUseBrushTransparency=false
        roundingType=HalfHeightRadius
        width=0
        color {colorUseRule=UseColor_Specified, specifiedColor=<rgba 0 0 0 0>}
        cornerRadii=<0 0 0 0>
       }
       tintColor {colorUseRule=UseColor_ColorTable, specifiedColor=<rgba 0 0 0 0>}
       uVRegion {bIsValid=false, max=<0 0>, min=<0 0>}
      }
      selectedBackgroundColor {colorUseRule=UseColor_ColorTable, specifiedColor=<rgba 0 0 0 0>}
      shadowColorAndOpacity=<rgba 0 0 0 1>
      shadowOffset=<0 0>
      strikeBrush {
       bIsDynamicallyLoaded=false
       drawAs=Image
       imageType=NoImage
       mirroring=NoMirror
       resourceName=None
       resourceObject=null
       tiling=NoTile
       imageSize=<32 32>
       margin=<ltrb 0 0 0 0>
       outlineSettings {
        bUseBrushTransparency=false
        roundingType=HalfHeightRadius
        width=0
        color {colorUseRule=UseColor_Specified, specifiedColor=<rgba 0 0 0 0>}
        cornerRadii=<0 0 0 0>
       }
       tintColor {colorUseRule=UseColor_Specified, specifiedColor=<rgba 1 1 1 1>}
       uVRegion {bIsValid=false, max=<0 0>, min=<0 0>}
      }
      underlineBrush {
       bIsDynamicallyLoaded=false
       drawAs=Image
       imageType=NoImage
       mirroring=NoMirror
       resourceName=None
       resourceObject=null
       tiling=NoTile
       imageSize=<32 32>
       margin=<ltrb 0 0 0 0>
       outlineSettings {
        bUseBrushTransparency=false
        roundingType=HalfHeightRadius
        width=0
        color {colorUseRule=UseColor_Specified, specifiedColor=<rgba 0 0 0 0>}
        cornerRadii=<0 0 0 0>
       }
       tintColor {colorUseRule=UseColor_Specified, specifiedColor=<rgba 1 1 1 1>}
       uVRegion {bIsValid=false, max=<0 0>, min=<0 0>}
      }
     }
     vScrollBarPadding=<ltrb 0 0 0 0>
    }}
   slot {class=HorizontalBoxSlot, properties {size {sizeRule=Fill, value=1}}}
  }
  btnAssignUserName {
   widgetClassPath=&TA_Generic01_Button
   parent=panelUserNameFilter
   widget=btnAssignUserName
   properties {
    toolTipText=Replace User Name field with your current sesions user name.
    buttonColor_Default=<rgba 1 0.5869 0.2027 1>
    iconBrush {
     bIsDynamicallyLoaded=false
     drawAs=Image
     imageType=FullColor
     mirroring=NoMirror
     resourceName=None
     resourceObject=/Game/Athena/UI/InGame/Creative/PropertyEditor/Icon_Picker/Textures_V2/64x64/T_UI_IconLibrary_Reset_64
     tiling=NoTile
     imageSize=<16 16>
     margin=<ltrb 0 0 0 0>
     outlineSettings {
      bUseBrushTransparency=false
      roundingType=HalfHeightRadius
      width=0
      color {colorUseRule=UseColor_Specified, specifiedColor=<rgba 0 0 0 0>}
      cornerRadii=<0 0 0 1>
     }
     tintColor {colorUseRule=UseColor_Specified, specifiedColor=<rgba 1 1 1 1>}
     uVRegion {bIsValid=false, max=<0 0>, min=<0 0>}
    }
   }
   slot {class=HorizontalBoxSlot, properties {verticalAlignment=VAlign_Center}}
  }
  HorizontalBox_1 {
   widgetClassPath=&HorizontalBox
   parent=StackBox_0
   widget=HorizontalBox_1
   slot {class=StackBoxSlot, properties {padding=<ltrb 0 5 0 5>}}
  }
  CommonTextBlock_67 {
   widgetClassPath=&CommonTextBlock
   parent=HorizontalBox_1
   widget=CommonTextBlock_67
   properties {
    justification=Right
    minDesiredWidth=125
    text="Workspace Filter: "
    font {
     bForceMonospaced=false
     bMaterialIsStencil=false
     fontMaterial=null
     fontObject=&NotoSans
     letterSpacing=0
     monospacedWidth=1
     size=10.5
     skewAmount=0
     typefaceFontName=NotoSans-Regular
     outlineSettings {
      bApplyOutlineToDropShadows=false
      bMiteredCorners=false
      bSeparateFillAlpha=false
      outlineMaterial=null
      outlineSize=0
      outlineColor=<rgba 0 0 0 1>
     }
    }
    shadowOffset=<0 0>
   }
   slot {class=HorizontalBoxSlot, properties {verticalAlignment=VAlign_Center, padding=<ltrb 10 0 5 0>}}
  }
  sComboClients {
   widgetClassPath=/Script/FortniteEditor.SearchableStringComboBox
   bIsVariable=true
   parent=HorizontalBox_1
   widget=sComboClients
   properties {
    font {
     bForceMonospaced=false
     bMaterialIsStencil=false
     fontMaterial=null
     fontObject=&Roboto
     letterSpacing=0
     monospacedWidth=1
     size=10.5
     skewAmount=0
     typefaceFontName=Bold
     outlineSettings {
      bApplyOutlineToDropShadows=false
      bMiteredCorners=false
      bSeparateFillAlpha=false
      outlineMaterial=null
      outlineSize=0
      outlineColor=<rgba 0 0 0 1>
     }
    }
    foregroundColor {colorUseRule=UseColor_Specified, specifiedColor=<rgba 0.5271 0.5271 0.5271 1>}
    itemStyle {
     bUseParentRowBrush=false
     activeBrush {
      bIsDynamicallyLoaded=false
      drawAs=Image
      imageType=NoImage
      mirroring=NoMirror
      resourceName=None
      resourceObject=null
      tiling=NoTile
      imageSize=<0 0>
      margin=<ltrb 0 0 0 0>
      outlineSettings {
       bUseBrushTransparency=false
       roundingType=HalfHeightRadius
       width=0
       color {colorUseRule=UseColor_Specified, specifiedColor=<rgba 0 0 0 0>}
       cornerRadii=<0 0 0 0>
      }
      tintColor {colorUseRule=UseColor_Specified, specifiedColor=<rgba 0 0.162 0.7454 1>}
      uVRegion {bIsValid=false, max=<0 0>, min=<0 0>}
     }
     activeHighlightedBrush {
      bIsDynamicallyLoaded=false
      drawAs=Image
      imageType=NoImage
      mirroring=NoMirror
      resourceName=None
      resourceObject=null
      tiling=NoTile
      imageSize=<0 0>
      margin=<ltrb 0 0 0 0>
      outlineSettings {
       bUseBrushTransparency=false
       roundingType=HalfHeightRadius
       width=0
       color {colorUseRule=UseColor_Specified, specifiedColor=<rgba 0 0 0 0>}
       cornerRadii=<0 0 0 0>
      }
      tintColor {colorUseRule=UseColor_Specified, specifiedColor=<rgba 0.0252 0.0319 0.0423 1>}
      uVRegion {bIsValid=false, max=<0 0>, min=<0 0>}
     }
     activeHoveredBrush {
      bIsDynamicallyLoaded=false
      drawAs=Image
      imageType=NoImage
      mirroring=NoMirror
      resourceName=None
      resourceObject=null
      tiling=NoTile
      imageSize=<0 0>
      margin=<ltrb 0 0 0 0>
      outlineSettings {
       bUseBrushTransparency=false
       roundingType=HalfHeightRadius
       width=0
       color {colorUseRule=UseColor_Specified, specifiedColor=<rgba 0 0 0 0>}
       cornerRadii=<0 0 0 0>
      }
      tintColor {colorUseRule=UseColor_Specified, specifiedColor=<rgba 0 0.162 0.7454 1>}
      uVRegion {bIsValid=false, max=<0 0>, min=<0 0>}
     }
     dropIndicator_Above {
      bIsDynamicallyLoaded=false
      drawAs=Box
      imageType=FullColor
      mirroring=NoMirror
      resourceName=../../../Engine/Content/Slate/Common/DropZoneIndicator_Above.png
      resourceObject=null
      tiling=NoTile
      imageSize=<0 0>
      margin=<ltrb 0.625 0.625 0 0>
      outlineSettings {
       bUseBrushTransparency=false
       roundingType=HalfHeightRadius
       width=0
       color {colorUseRule=UseColor_Specified, specifiedColor=<rgba 0 0 0 0>}
       cornerRadii=<0 0 0 0>
      }
      tintColor {colorUseRule=UseColor_Specified, specifiedColor=<rgba 0.0091 0.3515 0.9647 1>}
      uVRegion {bIsValid=false, max=<0 0>, min=<0 0>}
     }
     dropIndicator_Below {
      bIsDynamicallyLoaded=false
      drawAs=Box
      imageType=FullColor
      mirroring=NoMirror
      resourceName=../../../Engine/Content/Slate/Common/DropZoneIndicator_Below.png
      resourceObject=null
      tiling=NoTile
      imageSize=<0 0>
      margin=<ltrb 0.625 0 0 0.625>
      outlineSettings {
       bUseBrushTransparency=false
       roundingType=HalfHeightRadius
       width=0
       color {colorUseRule=UseColor_Specified, specifiedColor=<rgba 0 0 0 0>}
       cornerRadii=<0 0 0 0>
      }
      tintColor {colorUseRule=UseColor_Specified, specifiedColor=<rgba 0.0091 0.3515 0.9647 1>}
      uVRegion {bIsValid=false, max=<0 0>, min=<0 0>}
     }
     dropIndicator_Onto {
      bIsDynamicallyLoaded=false
      drawAs=Box
      imageType=FullColor
      mirroring=NoMirror
      resourceName=../../../Engine/Content/Slate/Common/DropZoneIndicator_Onto.png
      resourceObject=null
      tiling=NoTile
      imageSize=<0 0>
      margin=<ltrb 0.25 0.25 0.25 0.25>
      outlineSettings {
       bUseBrushTransparency=false
       roundingType=HalfHeightRadius
       width=0
       color {colorUseRule=UseColor_Specified, specifiedColor=<rgba 0 0 0 0>}
       cornerRadii=<0 0 0 0>
      }
      tintColor {colorUseRule=UseColor_Specified, specifiedColor=<rgba 0.0091 0.3515 0.9647 1>}
      uVRegion {bIsValid=false, max=<0 0>, min=<0 0>}
     }
     evenRowBackgroundBrush {
      bIsDynamicallyLoaded=false
      drawAs=Image
      imageType=NoImage
      mirroring=NoMirror
      resourceName=None
      resourceObject=null
      tiling=NoTile
      imageSize=<0 0>
      margin=<ltrb 0 0 0 0>
      outlineSettings {
       bUseBrushTransparency=false
       roundingType=HalfHeightRadius
       width=0
       color {colorUseRule=UseColor_Specified, specifiedColor=<rgba 0 0 0 0>}
       cornerRadii=<0 0 0 0>
      }
      tintColor {colorUseRule=UseColor_Specified, specifiedColor=<rgba 0.0103 0.0103 0.0103 1>}
      uVRegion {bIsValid=false, max=<0 0>, min=<0 0>}
     }
     evenRowBackgroundHoveredBrush {
      bIsDynamicallyLoaded=false
      drawAs=Image
      imageType=NoImage
      mirroring=NoMirror
      resourceName=None
      resourceObject=null
      tiling=NoTile
      imageSize=<0 0>
      margin=<ltrb 0 0 0 0>
      outlineSettings {
       bUseBrushTransparency=false
       roundingType=HalfHeightRadius
       width=0
       color {colorUseRule=UseColor_Specified, specifiedColor=<rgba 0 0 0 0>}
       cornerRadii=<0 0 0 0>
      }
      tintColor {colorUseRule=UseColor_Specified, specifiedColor=<rgba 0.0176 0.0176 0.0176 1>}
      uVRegion {bIsValid=false, max=<0 0>, min=<0 0>}
     }
     inactiveBrush {
      bIsDynamicallyLoaded=false
      drawAs=Image
      imageType=NoImage
      mirroring=NoMirror
      resourceName=None
      resourceObject=null
      tiling=NoTile
      imageSize=<0 0>
      margin=<ltrb 0 0 0 0>
      outlineSettings {
       bUseBrushTransparency=false
       roundingType=HalfHeightRadius
       width=0
       color {colorUseRule=UseColor_Specified, specifiedColor=<rgba 0 0 0 0>}
       cornerRadii=<0 0 0 0>
      }
      tintColor {colorUseRule=UseColor_Specified, specifiedColor=<rgba 0.0513 0.0953 0.159 1>}
      uVRegion {bIsValid=false, max=<0 0>, min=<0 0>}
     }
     inactiveHighlightedBrush {
      bIsDynamicallyLoaded=false
      drawAs=Image
      imageType=NoImage
      mirroring=NoMirror
      resourceName=None
      resourceObject=null
      tiling=NoTile
      imageSize=<0 0>
      margin=<ltrb 0 0 0 0>
      outlineSettings {
       bUseBrushTransparency=false
       roundingType=HalfHeightRadius
       width=0
       color {colorUseRule=UseColor_Specified, specifiedColor=<rgba 0 0 0 0>}
       cornerRadii=<0 0 0 0>
      }
      tintColor {colorUseRule=UseColor_Specified, specifiedColor=<rgba 0.0252 0.0319 0.0423 1>}
      uVRegion {bIsValid=false, max=<0 0>, min=<0 0>}
     }
     inactiveHoveredBrush {
      bIsDynamicallyLoaded=false
      drawAs=Image
      imageType=NoImage
      mirroring=NoMirror
      resourceName=None
      resourceObject=null
      tiling=NoTile
      imageSize=<0 0>
      margin=<ltrb 0 0 0 0>
      outlineSettings {
       bUseBrushTransparency=false
       roundingType=HalfHeightRadius
       width=0
       color {colorUseRule=UseColor_Specified, specifiedColor=<rgba 0 0 0 0>}
       cornerRadii=<0 0 0 0>
      }
      tintColor {colorUseRule=UseColor_Specified, specifiedColor=<rgba 0.0513 0.0953 0.159 1>}
      uVRegion {bIsValid=false, max=<0 0>, min=<0 0>}
     }
     oddRowBackgroundBrush {
      bIsDynamicallyLoaded=false
      drawAs=Image
      imageType=NoImage
      mirroring=NoMirror
      resourceName=None
      resourceObject=null
      tiling=NoTile
      imageSize=<0 0>
      margin=<ltrb 0 0 0 0>
      outlineSettings {
       bUseBrushTransparency=false
       roundingType=HalfHeightRadius
       width=0
       color {colorUseRule=UseColor_Specified, specifiedColor=<rgba 0 0 0 0>}
       cornerRadii=<0 0 0 0>
      }
      tintColor {colorUseRule=UseColor_Specified, specifiedColor=<rgba 0.0103 0.0103 0.0103 1>}
      uVRegion {bIsValid=false, max=<0 0>, min=<0 0>}
     }
     oddRowBackgroundHoveredBrush {
      bIsDynamicallyLoaded=false
      drawAs=Image
      imageType=NoImage
      mirroring=NoMirror
      resourceName=None
      resourceObject=null
      tiling=NoTile
      imageSize=<0 0>
      margin=<ltrb 0 0 0 0>
      outlineSettings {
       bUseBrushTransparency=false
       roundingType=HalfHeightRadius
       width=0
       color {colorUseRule=UseColor_Specified, specifiedColor=<rgba 0 0 0 0>}
       cornerRadii=<0 0 0 0>
      }
      tintColor {colorUseRule=UseColor_Specified, specifiedColor=<rgba 0.0176 0.0176 0.0176 1>}
      uVRegion {bIsValid=false, max=<0 0>, min=<0 0>}
     }
     parentRowBackgroundBrush {
      bIsDynamicallyLoaded=false
      drawAs=Image
      imageType=NoImage
      mirroring=NoMirror
      resourceName=None
      resourceObject=null
      tiling=NoTile
      imageSize=<32 32>
      margin=<ltrb 0 0 0 0>
      outlineSettings {
       bUseBrushTransparency=false
       roundingType=HalfHeightRadius
       width=0
       color {colorUseRule=UseColor_Specified, specifiedColor=<rgba 0 0 0 0>}
       cornerRadii=<0 0 0 0>
      }
      tintColor {colorUseRule=UseColor_Specified, specifiedColor=<rgba 1 1 1 1>}
      uVRegion {bIsValid=false, max=<0 0>, min=<0 0>}
     }
     parentRowBackgroundHoveredBrush {
      bIsDynamicallyLoaded=false
      drawAs=Image
      imageType=NoImage
      mirroring=NoMirror
      resourceName=None
      resourceObject=null
      tiling=NoTile
      imageSize=<32 32>
      margin=<ltrb 0 0 0 0>
      outlineSettings {
       bUseBrushTransparency=false
       roundingType=HalfHeightRadius
       width=0
       color {colorUseRule=UseColor_Specified, specifiedColor=<rgba 0 0 0 0>}
       cornerRadii=<0 0 0 0>
      }
      tintColor {colorUseRule=UseColor_Specified, specifiedColor=<rgba 1 1 1 1>}
      uVRegion {bIsValid=false, max=<0 0>, min=<0 0>}
     }
     selectedTextColor {colorUseRule=UseColor_Specified, specifiedColor=<rgba 1 1 1 1>}
     selectorFocusedBrush {
      bIsDynamicallyLoaded=false
      drawAs=Border
      imageType=FullColor
      mirroring=NoMirror
      resourceName=../../../Engine/Content/Slate/Common/Selector.png
      resourceObject=null
      tiling=Both
      imageSize=<0 0>
      margin=<ltrb 0.25 0.25 0.25 0.25>
      outlineSettings {
       bUseBrushTransparency=false
       roundingType=HalfHeightRadius
       width=0
       color {colorUseRule=UseColor_Specified, specifiedColor=<rgba 0 0 0 0>}
       cornerRadii=<0 0 0 0>
      }
      tintColor {colorUseRule=UseColor_Specified, specifiedColor=<rgba 0.701 0.225 0.003 1>}
      uVRegion {bIsValid=false, max=<0 0>, min=<0 0>}
     }
     textColor {colorUseRule=UseColor_Specified, specifiedColor=<rgba 0.5271 0.5271 0.5271 1>}
    }
    widgetStyle {
     comboButtonStyle {
      downArrowAlign=VAlign_Center
      buttonStyle {
       clickedSlateSound {resourceObject=null}
       disabled {
        bIsDynamicallyLoaded=false
        drawAs=RoundedBox
        imageType=NoImage
        mirroring=NoMirror
        resourceName=None
        resourceObject=null
        tiling=NoTile
        imageSize=<0 0>
        margin=<ltrb 0 0 0 0>
        outlineSettings {
         bUseBrushTransparency=false
         roundingType=FixedRadius
         width=1
         color {colorUseRule=UseColor_ColorTable, specifiedColor=<rgba 0 0 0 0>}
         cornerRadii=<4 4 4 4>
        }
        tintColor {colorUseRule=UseColor_Specified, specifiedColor=<rgba 0.0395 0.0395 0.0395 1>}
        uVRegion {bIsValid=false, max=<0 0>, min=<0 0>}
       }
       disabledForeground {colorUseRule=UseColor_ColorTable, specifiedColor=<rgba 0 0 0 0>}
       hovered {
        bIsDynamicallyLoaded=false
        drawAs=RoundedBox
        imageType=NoImage
        mirroring=NoMirror
        resourceName=None
        resourceObject=null
        tiling=NoTile
        imageSize=<0 0>
        margin=<ltrb 0 0 0 0>
        outlineSettings {
         bUseBrushTransparency=false
         roundingType=FixedRadius
         width=1
         color {colorUseRule=UseColor_ColorTable, specifiedColor=<rgba 0 0 0 0>}
         cornerRadii=<4 4 4 4>
        }
        tintColor {colorUseRule=UseColor_Specified, specifiedColor=<rgba 0.0048 0.0048 0.0048 1>}
        uVRegion {bIsValid=false, max=<0 0>, min=<0 0>}
       }
       hoveredForeground {colorUseRule=UseColor_ColorTable, specifiedColor=<rgba 0 0 0 0>}
       hoveredSlateSound {resourceObject=null}
       normal {
        bIsDynamicallyLoaded=false
        drawAs=RoundedBox
        imageType=NoImage
        mirroring=NoMirror
        resourceName=None
        resourceObject=null
        tiling=NoTile
        imageSize=<0 0>
        margin=<ltrb 0 0 0 0>
        outlineSettings {
         bUseBrushTransparency=false
         roundingType=FixedRadius
         width=1
         color {colorUseRule=UseColor_ColorTable, specifiedColor=<rgba 0 0 0 0>}
         cornerRadii=<4 4 4 4>
        }
        tintColor {colorUseRule=UseColor_Specified, specifiedColor=<rgba 0.0048 0.0048 0.0048 1>}
        uVRegion {bIsValid=false, max=<0 0>, min=<0 0>}
       }
       normalForeground {colorUseRule=UseColor_ColorTable, specifiedColor=<rgba 0 0 0 0>}
       normalPadding=<ltrb 8 1 8 1>
       pressed {
        bIsDynamicallyLoaded=false
        drawAs=RoundedBox
        imageType=NoImage
        mirroring=NoMirror
        resourceName=None
        resourceObject=null
        tiling=NoTile
        imageSize=<0 0>
        margin=<ltrb 0 0 0 0>
        outlineSettings {
         bUseBrushTransparency=false
         roundingType=FixedRadius
         width=1
         color {colorUseRule=UseColor_ColorTable, specifiedColor=<rgba 0 0 0 0>}
         cornerRadii=<4 4 4 4>
        }
        tintColor {colorUseRule=UseColor_Specified, specifiedColor=<rgba 0.0048 0.0048 0.0048 1>}
        uVRegion {bIsValid=false, max=<0 0>, min=<0 0>}
       }
       pressedForeground {colorUseRule=UseColor_ColorTable, specifiedColor=<rgba 0 0 0 0>}
       pressedPadding=<ltrb 8 2 8 0>
       pressedSlateSound {resourceObject=null}
      }
      contentPadding=<ltrb 0 0 0 0>
      downArrowImage {
       bIsDynamicallyLoaded=false
       drawAs=Image
       imageType=Vector
       mirroring=NoMirror
       resourceName=../../../Engine/Content/Slate/Starship/CoreWidgets/ComboBox/wide-chevron-down.svg
       resourceObject=null
       tiling=NoTile
       imageSize=<16 16>
       margin=<ltrb 0 0 0 0>
       outlineSettings {
        bUseBrushTransparency=false
        roundingType=HalfHeightRadius
        width=0
        color {colorUseRule=UseColor_Specified, specifiedColor=<rgba 0 0 0 0>}
        cornerRadii=<0 0 0 0>
       }
       tintColor {colorUseRule=UseColor_Specified, specifiedColor=<rgba 1 1 1 1>}
       uVRegion {bIsValid=false, max=<0 0>, min=<0 0>}
      }
      downArrowPadding=<ltrb 2 2 2 2>
      menuBorderBrush {
       bIsDynamicallyLoaded=false
       drawAs=RoundedBox
       imageType=NoImage
       mirroring=NoMirror
       resourceName=None
       resourceObject=null
       tiling=NoTile
       imageSize=<0 0>
       margin=<ltrb 0 0 0 0>
       outlineSettings {
        bUseBrushTransparency=false
        roundingType=FixedRadius
        width=1
        color {colorUseRule=UseColor_ColorTable, specifiedColor=<rgba 0 0 0 0>}
        cornerRadii=<0 0 0 0>
       }
       tintColor {colorUseRule=UseColor_Specified, specifiedColor=<rgba 0.0395 0.0395 0.0395 1>}
       uVRegion {bIsValid=false, max=<0 0>, min=<0 0>}
      }
      menuBorderPadding=<ltrb 0 0 0 0>
      shadowColorAndOpacity=<rgba 0 0 0 1>
      shadowOffset=<0 0>
     }
     contentPadding=<ltrb 0 0 0 0>
     menuRowPadding=<ltrb 8 3 8 3>
     pressedSlateSound {resourceObject=null}
     selectionChangeSlateSound {resourceObject=null}
    }
   }
   slot {class=HorizontalBoxSlot, properties {size {sizeRule=Fill, value=1}}}
  }
  HorizontalBox_0 {
   widgetClassPath=&HorizontalBox
   parent=StackBox_0
   widget=HorizontalBox_0
   slot {class=StackBoxSlot, properties {horizontalAlignment=HAlign_Center}}
  }
  timePickerEarliest {
   widgetClassPath=&/EUW_TimePicker
   bIsVariable=true
   parent=HorizontalBox_0
   widget=timePickerEarliest
   properties {label=EARLIEST, startingDate=2024.01.01-04.00.00}
   slot {class=HorizontalBoxSlot, properties {padding=<ltrb 0 0 10 5>}}
  }
  Image_528 {
   widgetClassPath=&Image
   bIsVariable=true
   parent=HorizontalBox_0
   widget=Image_528
   properties {
    brush {
     bIsDynamicallyLoaded=false
     drawAs=Image
     imageType=NoImage
     mirroring=NoMirror
     resourceName=None
     resourceObject=null
     tiling=NoTile
     imageSize=<12 11>
     margin=<ltrb 0 0 0 0>
     outlineSettings {
      bUseBrushTransparency=false
      roundingType=HalfHeightRadius
      width=0
      color {colorUseRule=UseColor_Specified, specifiedColor=<rgba 0 0 0 0>}
      cornerRadii=<0 0 0 0>
     }
     tintColor {colorUseRule=UseColor_Specified, specifiedColor=<rgba 1 1 1 1>}
     uVRegion {bIsValid=false, max=<0 0>, min=<0 0>}
    }
    colorAndOpacity=<rgba 0.7333 0.7333 0.7333 1>
   }
   slot {class=HorizontalBoxSlot, properties {verticalAlignment=VAlign_Center}}
  }
  Image {
   widgetClassPath=&Image
   bIsVariable=true
   parent=HorizontalBox_0
   widget=Image
   properties {
    brush {
     bIsDynamicallyLoaded=false
     drawAs=Image
     imageType=NoImage
     mirroring=NoMirror
     resourceName=None
     resourceObject=null
     tiling=NoTile
     imageSize=<4 62>
     margin=<ltrb 0 0 0 0>
     outlineSettings {
      bUseBrushTransparency=false
      roundingType=HalfHeightRadius
      width=0
      color {colorUseRule=UseColor_Specified, specifiedColor=<rgba 0 0 0 0>}
      cornerRadii=<0 0 0 0>
     }
     tintColor {colorUseRule=UseColor_Specified, specifiedColor=<rgba 1 1 1 1>}
     uVRegion {bIsValid=false, max=<0 0>, min=<0 0>}
    }
    colorAndOpacity=<rgba 0.7333 0.7333 0.7333 1>
   }
   slot {class=HorizontalBoxSlot, properties {verticalAlignment=VAlign_Bottom, padding=<ltrb 5 0 5 0>}}
  }
  Image_427 {
   widgetClassPath=&Image
   bIsVariable=true
   parent=HorizontalBox_0
   widget=Image_427
   properties {
    brush {
     bIsDynamicallyLoaded=false
     drawAs=Image
     imageType=FullColor
     mirroring=NoMirror
     resourceName=None
     resourceObject=/Game/NeverCook/EditorUtilities/DebugTools/AssetChangeFinder/Art/ArrowRight
     tiling=NoTile
     imageSize=<32 32>
     margin=<ltrb 0 0 0 0>
     outlineSettings {
      bUseBrushTransparency=false
      roundingType=HalfHeightRadius
      width=0
      color {colorUseRule=UseColor_Specified, specifiedColor=<rgba 0 0 0 0>}
      cornerRadii=<0 0 0 0>
     }
     tintColor {colorUseRule=UseColor_Specified, specifiedColor=<rgba 1 1 1 1>}
     uVRegion {bIsValid=false, max=<0 0>, min=<0 0>}
    }
    colorAndOpacity=<rgba 0.7333 0.7333 0.7333 1>
   }
   slot {class=HorizontalBoxSlot, properties {horizontalAlignment=HAlign_Center, verticalAlignment=VAlign_Center}}
  }
  timePickerLatest {
   widgetClassPath=&EUW_TimePicker
   bIsVariable=true
   parent=HorizontalBox_0
   widget=timePickerLatest
   properties {label=LATEST, showNowButton=true, startingDate=2024.02.01-04.00.00}
   slot {class=HorizontalBoxSlot, properties {padding=<ltrb 10 0 0 0>}}
  }
  Border_165 {
   widgetClassPath=&Border
   parent=StackBox_0
   widget=Border_165
   properties {
    background {
     bIsDynamicallyLoaded=false
     drawAs=Border
     imageType=FullColor
     mirroring=NoMirror
     resourceName=None
     resourceObject=null
     tiling=NoTile
     imageSize=<128 128>
     margin=<ltrb 0.4 0.4 0.4 0.4>
     outlineSettings {
      bUseBrushTransparency=false
      roundingType=HalfHeightRadius
      width=0
      color {colorUseRule=UseColor_Specified, specifiedColor=<rgba 0 0 0 0>}
      cornerRadii=<0 0 0 0>
     }
     tintColor {colorUseRule=UseColor_Specified, specifiedColor=<rgba 1 1 1 1>}
     uVRegion {bIsValid=false, max=<0 0>, min=<0 0>}
    }
    brushColor=<rgba 0.7333 0.7333 0.7333 1>
    padding=<ltrb 6 6 6 6>
   }
   slot {class=StackBoxSlot, properties {size {sizeRule=Fill, value=1}}}
  }
  TA_Generic01_List {
   widgetClassPath=/Game/Blueprints/Blutilities/ForTechArt/Widgets/List/TA_Generic01_List
   bIsVariable=true
   parent=Border_165
   widget=TA_Generic01_List
   properties {
    allowMultiSelect=false
    listItemClass=&/UW_AssetChangeFinder_ListItem.UW_AssetChangeFinder_ListItem_C
   }
   slot {class=BorderSlot, properties {padding=<ltrb 6 6 6 6>}}
  }
  hBoxLastrun {
   widgetClassPath=&HorizontalBox
   bIsVariable=true
   parent=StackBox_0
   widget=hBoxLastrun
   properties {visibility=Hidden}
   slot {class=StackBoxSlot, properties {horizontalAlignment=HAlign_Center, verticalAlignment=VAlign_Bottom, padding=<ltrb 0 8 0 8>}}
  }
  CommonTextBlock {
   widgetClassPath=&CommonTextBlock
   bIsVariable=true
   parent=hBoxLastrun
   widget=CommonTextBlock
   properties {
    text="Workspace: "
    font {
     bForceMonospaced=false
     bMaterialIsStencil=false
     fontMaterial=null
     fontObject=&NotoSans
     letterSpacing=0
     monospacedWidth=1
     size=10.5
     skewAmount=0
     typefaceFontName=NotoSans-Regular
     outlineSettings {
      bApplyOutlineToDropShadows=false
      bMiteredCorners=false
      bSeparateFillAlpha=false
      outlineMaterial=null
      outlineSize=0
      outlineColor=<rgba 0 0 0 1>
     }
    }
    shadowOffset=<0 0>
   }
   slot {class=HorizontalBoxSlot, properties {verticalAlignment=VAlign_Center, padding=<ltrb 10 0 5 0>}}
  }
  btnFind {
   widgetClassPath=&TA_Generic01_Button
   parent=hBoxLastrun
   widget=btnFind
   properties {iconBrush {
     bIsDynamicallyLoaded=false
     drawAs=Image
     imageType=FullColor
     mirroring=NoMirror
     resourceName=None
     resourceObject=/Game/Athena/UI/Challenges/Art/ChallengeTypeIcons/UI_ChallengeIcon_Find
     tiling=NoTile
     imageSize=<16 16>
     margin=<ltrb 0 0 0 0>
     outlineSettings {
      bUseBrushTransparency=false
      roundingType=HalfHeightRadius
      width=0
      color {colorUseRule=UseColor_Specified, specifiedColor=<rgba 0 0 0 0>}
      cornerRadii=<0 0 0 1>
     }
     tintColor {colorUseRule=UseColor_Specified, specifiedColor=<rgba 1 1 1 1>}
     uVRegion {bIsValid=false, max=<0 0>, min=<0 0>}
    }}
   slot {class=HorizontalBoxSlot}
  }
  HorizontalBox_3 {
   widgetClassPath=&HorizontalBox
   parent=StackBox_0
   widget=HorizontalBox_3
   slot {class=StackBoxSlot, properties {verticalAlignment=VAlign_Bottom}}
  }
  btnRunAssetFilter {
   widgetClassPath=&TA_Generic01_Button
   parent=HorizontalBox_3
   widget=btnRunAssetFilter
   properties {label=RUN, sizeOverride=<0 30>}
   slot {class=HorizontalBoxSlot, properties {verticalAlignment=VAlign_Center, size {sizeRule=Fill, value=1}}}
  }
  Border_172 {
   widgetClassPath=&Border
   parent=HorizontalBox_3
   widget=Border_172
   properties {brushColor=<rgba 0 0 0 0.5>}
   slot {class=HorizontalBoxSlot}
  }
  HorizontalBox_2 {widgetClassPath=&HorizontalBox, parent=Border_172, widget=HorizontalBox_2, slot {class=BorderSlot}}
  cbUserFilter {
   widgetClassPath=/Game/Blueprints/Blutilities/ForTechArt/Widgets/Buttons/TA_Generic01_Checkbox
   bIsVariable=true
   parent=HorizontalBox_2
   widget=cbUserFilter
   properties {
    label=User Filter
    iconStyle {
     bIsDynamicallyLoaded=false
     drawAs=Image
     imageType=NoImage
     mirroring=NoMirror
     resourceName=None
     resourceObject=null
     tiling=NoTile
     imageSize=<32 32>
     margin=<ltrb 0 0 0 0>
     outlineSettings {
      bUseBrushTransparency=false
      roundingType=HalfHeightRadius
      width=0
      color {colorUseRule=UseColor_Specified, specifiedColor=<rgba 0 0 0 0>}
      cornerRadii=<0 0 0 1>
     }
     tintColor {colorUseRule=UseColor_Specified, specifiedColor=<rgba 1 1 1 1>}
     uVRegion {bIsValid=false, max=<0 0>, min=<0 0>}
    }
   }
   slot {class=HorizontalBoxSlot, properties {verticalAlignment=VAlign_Center}}
  }
  cbAssetFilter {
   widgetClassPath=&TA_Generic01_Checkbox
   bIsVariable=true
   parent=HorizontalBox_2
   widget=cbAssetFilter
   properties {isSelected=true, label=Asset Filter}
   slot {class=HorizontalBoxSlot, properties {verticalAlignment=VAlign_Center, padding=<ltrb 5 0 0 0>}}
  }
  btnHelp {
   widgetClassPath=&TA_Generic01_Button
   parent=HorizontalBox_3
   widget=btnHelp
   properties {
    buttonColor_Default=<rgba 0.6719 0.5065 0.0482 1>
    iconBrush {
     bIsDynamicallyLoaded=false
     drawAs=Image
     imageType=FullColor
     mirroring=NoMirror
     resourceName=None
     resourceObject=/Game/FrontEnd/Store/Cards/Assets/T_IconScreen_Question
     tiling=NoTile
     imageSize=<24 24>
     margin=<ltrb 0 0 0 0>
     outlineSettings {
      bUseBrushTransparency=false
      roundingType=HalfHeightRadius
      width=0
      color {colorUseRule=UseColor_Specified, specifiedColor=<rgba 0 0 0 0>}
      cornerRadii=<0 0 0 1>
     }
     tintColor {colorUseRule=UseColor_Specified, specifiedColor=<rgba 1 1 1 1>}
     uVRegion {bIsValid=false, max=<0 0>, min=<0 0>}
    }
   }
   slot {class=HorizontalBoxSlot, properties {verticalAlignment=VAlign_Center}}
  }
  panelUnpipeFailed {
   widgetClassPath=&Border
   bIsVariable=true
   parent=Overlay_0
   widget=panelUnpipeFailed
   properties {visibility=Collapsed, brushColor=<rgba 0.026 0.026 0.026 0.962>, padding=<ltrb 0 0 0 0>}
   slot {class=OverlaySlot, properties {horizontalAlignment=HAlign_Fill, verticalAlignment=VAlign_Fill}}
  }
  VerticalBox_85 {
   widgetClassPath=&VerticalBox
   parent=panelUnpipeFailed
   widget=VerticalBox_85
   slot {class=BorderSlot, properties {padding=<ltrb 0 0 0 0>}}
  }
  CommonTextBlock_2 {
   widgetClassPath=&CommonTextBlock
   parent=VerticalBox_85
   widget=CommonTextBlock_2
   properties {
    autoWrapText=true
    minDesiredWidth=125
    text="Could not connect to Unpipe. Enable pipeline tools in the main menu and click Try Again. "
    font {
     bForceMonospaced=false
     bMaterialIsStencil=false
     fontMaterial=null
     fontObject=&NotoSans
     letterSpacing=0
     monospacedWidth=1
     size=18
     skewAmount=0
     typefaceFontName=NotoSans-Regular
     outlineSettings {
      bApplyOutlineToDropShadows=false
      bMiteredCorners=false
      bSeparateFillAlpha=false
      outlineMaterial=null
      outlineSize=0
      outlineColor=<rgba 0 0 0 1>
     }
    }
    shadowOffset=<0 0>
   }
   slot {class=VerticalBoxSlot, properties {horizontalAlignment=HAlign_Center, verticalAlignment=VAlign_Center, padding=<ltrb 15 15 15 15>}}
  }
  ScaleBox_0 {
   widgetClassPath=&ScaleBox
   parent=VerticalBox_85
   widget=ScaleBox_0
   slot {
    class=VerticalBoxSlot
    properties {horizontalAlignment=HAlign_Center, verticalAlignment=VAlign_Top, padding=<ltrb 10 5 10 5>, size {sizeRule=Fill, value=1}}
   }
  }
  Image_320 {
   widgetClassPath=&Image
   bIsVariable=true
   parent=ScaleBox_0
   widget=Image_320
   properties {brush {
     bIsDynamicallyLoaded=false
     drawAs=Image
     imageType=FullColor
     mirroring=NoMirror
     resourceName=None
     resourceObject=/Game/NeverCook/EditorUtilities/DebugTools/AssetChangeFinder/Art/T_PipelineLocation
     tiling=NoTile
     imageSize=<805 67>
     margin=<ltrb 0 0 0 0>
     outlineSettings {
      bUseBrushTransparency=false
      roundingType=HalfHeightRadius
      width=0
      color {colorUseRule=UseColor_Specified, specifiedColor=<rgba 0 0 0 0>}
      cornerRadii=<0 0 0 0>
     }
     tintColor {colorUseRule=UseColor_Specified, specifiedColor=<rgba 1 1 1 1>}
     uVRegion {bIsValid=false, max=<0 0>, min=<0 0>}
    }}
   slot {class=ScaleBoxSlot}
  }
  btnRetryInit {
   widgetClassPath=&TA_Generic01_Button
   parent=VerticalBox_85
   widget=btnRetryInit
   properties {label=Try Again}
   slot {class=VerticalBoxSlot}
  }
 ]
}
